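# Material function for the Asteria ice cave rock ice, outputting ice material attributes plus Noise and RVT Snow masks.
# World-aligned noise and rock detail textures supply normals, with an extra blend at Epic quality, and the snow mask height-lerps over world Z.
# Glow combines location-specific sphere masks, time of day, light direction and height, but emissive is zeroed before output.
# Several branches, including time-of-day emissive tinting and parallax noise, are left unconnected.
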MaterialFunction /Game/Environments/Asteria/Rocks/Materials/MaterialFunctions/MF_IceCave_Ice
graph {nodes [
  n1 = VectorParameter() {group=Ice, parameterName=IceTone Emissive Inside, sortPriority=3, defaultValue=<rgba 0 0.088 0.1997 0.1>}
  n2 = NamedRerouteUsage()
  n3 = VectorParameter() {group=Ice, parameterName=IceTone Emissive Outside, sortPriority=2, defaultValue=<rgba 0 0.3248 0.5156 0.5>}
  n4 = LinearInterpolate(A=@n1.RGB, B=@n3.RGB, Alpha=NamedRerouteUsage())
  n5 = ScalarParameter() {defaultValue=1, parameterName=SSS Value}
  n6 = ScalarParameter() {defaultValue=0.25, parameterName=Central Glow}
  n7 = SphereMask(
   A=WorldPosition().XYZ
   B=VectorParameter() {parameterName=Sphere big, defaultValue=<rgba 97327.7656 -16384 2688 1>}.RGB
   Radius=ScalarParameter() {defaultValue=7500, parameterName=SS}
   Hardness=ScalarParameter() {defaultValue=0.5, parameterName=SH}
  )
  n8 = Constant4Vector() {constant=<rgba 1 1 0 1>}
  n9 = CollectionParameter() {
   collection=&FortniteMaterialParameters
   parameterName=FortniteActiveTimeOfDay
  }
  n10 = Multiply(A=Add(A=Multiply(A=@n6, B=@n7), B=NamedRerouteUsage()), B=Saturate(None=DotProduct(A=@n8.RGBA, B=@n9)))
  n11 = Constant4Vector() {constant=<rgba 0 0 1 0>}
  n12 = Multiply(
   A=SphereMask(
    A=WorldPosition().XYZ
    B=VectorParameter() {parameterName=Sphere big 2, defaultValue=<rgba 104168 -18000 7932 1>}.RGB
    Radius=ScalarParameter() {defaultValue=12000, parameterName=SSBig2}
    Hardness=ScalarParameter() {parameterName=SHBig2}
   )
   B=Saturate(None=DotProduct(A=@n11.RGBA, B=@n9))
  )
  n13 = Multiply(A=NamedRerouteUsage(), B=Saturate(None=Add(A=@n10, B=@n12)))
  n14 = VertexNormalWS()
  n15 = Constant() {r=-0.5}
  n16 = CameraVectorWS()
  n17 = Divide(A=DotProduct(A=Add(A=@n14, B=@n15), B=@n16), B=Constant() {r=8})
  n18 = NamedRerouteDeclaration(None=@n17) {name=LightFresnel, nodeColor=<rgba 0 0.0156 1 1>}
  n19 = DotProduct(A=Multiply(A=AtmosphericLightVector()) {constB=-1}, B=@n16)
  n20 = Power(Base=@n19, Exp=Constant() {r=2})
  n21 = Constant() {r=1.5}
  n22 = Saturate(None=Multiply(A=LinearInterpolate(Alpha=@n18) {constA=0.5}, B=Multiply(A=@n20, B=@n21)))
  n23 = LinearInterpolate(
   A=ShadingPathSwitch(Default=ScalarParameter() {defaultValue=0.1, group=Ice, parameterName=E Min}, Mobile=Constant() {r=1})
   Alpha=Saturate(None=Multiply(A=@n13, B=@n22))
  )
  n24 = WorldPosition()
  n25 = Divide(A=Subtract(A=ComponentMask(None=@n24.XYZ) {b=true}, B=Constant() {r=2675.5405}), B=Constant() {r=3209.8735})
  n26 = DotProduct(A=CameraVectorWS(), B=VertexNormalWS())
  n27 = TextureObject() {samplerType=SAMPLERTYPE_Masks, texture=/Game/Environments/Asteria/Terrain/Textures/T_Asteria_Ice_Noise_M}
  n28 = ScalarParameter() {defaultValue=0.75, group=Ice, parameterName=ScaleN, sortPriority=1}
  n29 = MaterialFunctionCall!(
   Diffuse Texture Object
   Spec Texture Object=@n27
   TextureSize=Multiply(A=Constant() {r=800}, B=@n28)
   --------------------------
   Use seperate settings for Z projection=StaticBool() {value=true}
   Use seperate texture for Z projection
   Texture Size Z projection=ScalarParameter() {defaultValue=512, parameterName=Up Noise out}
   Z Plane Diffuse Texture Object
   Z Plane Spec Texture Object=@n27
   "--------------------------      "
   ProjectionTransitionContrast
   Use Texture Alpha Heightmap to Refine the Transistion
   HeightLerpTransitionContrast
   PickTheDiffuseHeightChannel
   TheDiffuseHeightChannel
   "--------------------------       "
   NormalTextureObject=TextureObject() {samplerType=SAMPLERTYPE_Normal, texture=/Game/Environments/Asteria/Terrain/Textures/T_Asteria_Ice_SubtleNoise_N}
   World Space Normals=StaticBool()
   "--------------------------  "
   Export Float 4 Diffuse
   Export Float 4 Spec
   "--------------------------     "
   WorldPosition
   World Space Normal
  ) {materialFunction=/Engine/Functions/Engine_MaterialFunctions02/Texturing/WorldAlignedTextures_Complex}
  n30 = Constant() {r=0.5}
  n31 = Constant2Vector() {g=-1, r=-1}
  n32 = MaterialFunctionCall(
   Spec Texture Object=TextureObject() {samplerType=SAMPLERTYPE_Masks, texture=/Game/Environments/Asteria/Rocks/Textures/T_Rock_BigDetails_M}
   TextureSize=Multiply(A=@n28, B=Constant() {r=2000})
   Texture Size Z projection=NamedRerouteUsage()
   NormalTextureObject=TextureObject() {samplerType=SAMPLERTYPE_Normal, texture=/Game/Environments/Asteria/Rocks/Textures/T_Rock_BigDetails_N}
   World Space Normals=StaticBool()
  ) {materialFunction=&WorldAlignedTextures_Complex}
  n33 = AppendVector(
   A=Multiply(A=Multiply(A=@n30, B=@n31), B=ComponentMask(None=@n32.Normal XYZ Texture) {g=true, r=true})
   B=ComponentMask(None=@n32.Normal XYZ Texture) {b=true}
  )
  n34 = Constant3Vector() {constant=<rgba 1 1 0 0>}
  n35 = Add(A=Multiply(A=ScalarParameter() {defaultValue=3, parameterName=Noise2}, B=Multiply(A=@n29.Normal XYZ Flat Top, B=@n34)), B=@n33)
  n36 = Saturate(None=LinearInterpolate(A=Constant() {r=-0.408}, B=Constant() {r=1}, Alpha=NamedRerouteUsage()))
  n37 = SetMaterialAttributes!(
   MaterialAttributes
   BaseColor=Multiply(A=@n1.RGB, B=@n1.RGB)
   Specular=Multiply(A=OneMinus(None=@n2), B=ScalarParameter() {defaultValue=0.5, group=Ice, parameterName=Ice Spec})
   Roughness=Constant() {r=0.1}
   SubsurfaceColor=NamedRerouteDeclaration(None=Multiply(A=@n4, B=@n5)) {name=E, nodeColor=<rgba 0 0.2266 1 1>}
   Normal=QualitySwitch(Default=@n29.Normal XYZ Flat Top, Epic=LinearInterpolate(A=@n33, B=@n35, Alpha=@n36))
   Opacity=Constant() {r=0.5}
   EmissiveColor=Multiply(A=Multiply(A=@n23, B=@n25), B=Saturate(None=@n26))
  ) {attributeSetTypes [
    #1
    #2
    #3
    #4
    #5
    #6
    #7
   ]}
  n38 = GetMaterialAttributes(None=@n37) {attributeGetTypes [#7]}
  n39 = GetMaterialAttributes(None=SetMaterialAttributes!(MaterialAttributes=@n38.MaterialAttributes, EmissiveColor=Constant()) {attributeSetTypes [#7]}) {attributeGetTypes [#4]}
  FunctionOutput(None=SetMaterialAttributes!(MaterialAttributes=@n39.MaterialAttributes, BaseColor=@n39.MaterialAttributes) {attributeSetTypes [#1]})
  FunctionOutput(None=ComponentMask(None=@n29.Z Spec) {r=true}) {outputName=Noise}
  n40 = WorldPosition()
  n41 = Min(
   A=Divide(A=ComponentMask(None=@n40.XYZ) {b=true}, B=ScalarParameter() {defaultValue=1000, group=Ice, parameterName=Ice to Snow Fade, sortPriority=4})
   B=NamedRerouteUsage()
  )
  n42 = Multiply(A=NamedRerouteUsage(), B=NamedRerouteUsage())
  FunctionOutput(None=Saturate(None=MaterialFunctionCall(Transition Phase=@n41, Height Texture=@n42) {materialFunction=&HeightLerp}.Lerp Alpha No Contrast)) {outputName=RVT Snow}
  n43 = TextureSampleParameter2D() {
   group=Ice Fallback
   parameterName=IceMask G8
   samplerType=SAMPLERTYPE_LinearGrayscale
   sortPriority=1
   texture=/Game/Environments/Asteria/Terrain/Frozen/Textures/T_Asteria_Frozen_Cliff_A_THK
  }
  MaterialFunctionCall(In=@n43.R, Power=ScalarParameter() {defaultValue=1, parameterName=Scurve}) {materialFunction=&SCurve}
  MaterialFunctionCall() {materialFunction=/Engine/Functions/Engine_MaterialFunctions02/WorldPositionOffset/CameraOffset}
  Multiply(A=@n19, B=@n19)
  n44 = TextureObject() {samplerType=SAMPLERTYPE_Masks, texture=&T_Asteria_Ice_Noise_M}
  n45 = ScalarParameter() {defaultValue=2000}
  n46 = Add(A=WorldPosition().XYZ, B=Multiply(A=CameraPositionWS()) {constB=0.5})
  n47 = Fresnel()
  Multiply(
   A=@n39.SubsurfaceColor
   B=Multiply(
    A=MaterialFunctionCall(TextureObject=@n44, TextureSize=@n45, WorldPosition=@n46) {materialFunction=&WorldAlignedTexture}.XYZ Texture
    B=OneMinus(None=@n47)
   )
  )
  n48 = Multiply(A=TextureCoordinate(), B=ScalarParameter() {defaultValue=2, parameterName=Param_2})
  n49 = ScalarParameter() {defaultValue=2, parameterName=Param_3}
  n50 = NamedRerouteUsage()
  Multiply(
   A=TextureSample(UVs=BumpOffset(Coordinate=@n48, Height=@n49)) {samplerType=SAMPLERTYPE_Masks, texture=&T_Asteria_Ice_Noise_M}.R
   B=@n50
  )
  n51 = NamedRerouteUsage()
  n52 = ShadingPathSwitch(Default=ScalarParameter() {defaultValue=0.65, parameterName=Desat color before emissive}, Mobile=Constant() {r=1})
  n53 = VectorParameter() {
   desc="Mult Emissive by This color per TOD "
   group=A_Emissive_TOD
   parameterName=EmissiveMorning
   sortPriority=1
   defaultValue=<rgba 0.4399 0.5451 0.1448 1>
  }
  n54 = VectorParameter() {
   desc="Mult Emissive by This color per TOD "
   group=A_Emissive_TOD
   parameterName=EmissiveDay
   sortPriority=2
   defaultValue=<rgba 0.4323 0.6776 1 1>
  }
  n55 = VectorParameter() {
   desc="Mult Emissive by This color per TOD "
   group=A_Emissive_TOD
   parameterName=EmissiveEvening
   sortPriority=3
   defaultValue=<rgba 0.8177 0.1303 0.1493 1>
  }
  n56 = VectorParameter() {
   desc="Mult Emissive by This color per TOD "
   group=A_Emissive_TOD
   parameterName=EmissiveNight
   sortPriority=4
   defaultValue=<rgba 0.3401 0.7534 1.5 0>
  }
  n57 = MaterialFunctionCall(Color_Morning=@n53.RGB, Color_Day=@n54.RGB, Color_Evening=@n55.RGB, Color_Night=@n56.RGB) {materialFunction=&Color_Time_of_Day}
  n58 = VectorParameter() {
   desc="Mult Emissive by This color per TOD "
   group=A_Emissive_TOD
   parameterName=Mob EmissiveMorning
   sortPriority=1
   defaultValue=<rgba 0.1334 0.1549 0.0421 1>
  }
  n59 = VectorParameter() {
   desc="Mult Emissive by This color per TOD "
   group=A_Emissive_TOD
   parameterName=Mob EmissiveDay
   sortPriority=2
   defaultValue=<rgba 0.0814 0.3247 0.6445 1>
  }
  n60 = VectorParameter() {
   desc="Mult Emissive by This color per TOD "
   group=A_Emissive_TOD
   parameterName=Mob EmissiveEvening
   sortPriority=3
   defaultValue=<rgba 0.194 0.1141 0.0599 1>
  }
  n61 = VectorParameter() {
   desc="Mult Emissive by This color per TOD "
   group=A_Emissive_TOD
   parameterName=Mob EmissiveNight
   sortPriority=4
   defaultValue=<rgba 0.1544 0.342 0.681 0>
  }
  n62 = MaterialFunctionCall(Color_Morning=@n58.RGB, Color_Day=@n59.RGB, Color_Evening=@n60.RGB, Color_Night=@n61.RGB) {materialFunction=&Color_Time_of_Day}
  Multiply(
   A=@n38.EmissiveColor
   B=MaterialFunctionCall(Base=Desaturation(None=@n51, Fraction=@n52), Blend=ShadingPathSwitch(Default=@n57.Result, Mobile=@n62.Result)) {materialFunction=&Blend_Overlay}.Result
  )
  Multiply(A=@n57.Result, B=ScalarParameter() {defaultValue=5, parameterName=Param_1})
  n63 = Constant() {r=1}
  n64 = ShadingPathSwitch(
   Default=DistanceFieldApproxAO!(
    World Position
    Normal
    BaseDistance=ScalarParameter() {defaultValue=75, group=Ice, parameterName=AO Dist, sortPriority=5}
    Radius=ScalarParameter() {defaultValue=500, group=Ice, parameterName=AO Rad, sortPriority=6}
   ) {numSteps=2}
   Forward=@n63
   Mobile=@n63
  )
  n65 = ScalarParameter() {defaultValue=0.5, group=Ice, parameterName=AO Min value, sortPriority=7}
  NamedRerouteDeclaration(None=QualitySwitch(Default=Constant() {r=1}, Epic=Max(A=@n64, B=@n65))) {name=AO, nodeColor=<rgba 0.0781 0 1 1>}
  NamedRerouteDeclaration(None=ComponentMask(None=@n29.XYZ Spec) {g=true}) {name=Noise, nodeColor=<rgba 0.8516 0 1 1>}
  NamedRerouteDeclaration(None=ComponentMask(None=@n32.XYZ Spec) {r=true}) {name=Noise Big, nodeColor=<rgba 0.0547 0 1 1>}
  n66 = WorldPosition()
  n67 = Constant3Vector() {constant=<rgba 97166.1875 -29021.6855 4868.0786 0>}
  n68 = ScalarParameter() {defaultValue=6500, parameterName=SS3}
  n69 = ScalarParameter() {defaultValue=0.25, parameterName=SH3}
  n70 = WorldPosition()
  n71 = Constant3Vector() {constant=<rgba 89359.0938 -7499.1738 3891.2168 0>}
  n72 = ScalarParameter() {defaultValue=4500, parameterName=SS4}
  n73 = ScalarParameter() {parameterName=SH4}
  n74 = Add(
   A=SphereMask(
    A=WorldPosition().XYZ
    B=Constant3Vector() {constant=<rgba 91447.8828 -17514.6387 8566.2998 0>}
    Radius=ScalarParameter() {defaultValue=6500, parameterName=SS2}
    Hardness=ScalarParameter() {parameterName=SH2}
   )
   B=Add(A=SphereMask(A=@n66.XYZ, B=@n67, Radius=@n68, Hardness=@n69), B=SphereMask(A=@n70.XYZ, B=@n71, Radius=@n72, Hardness=@n73))
  )
  NamedRerouteDeclaration(None=@n74) {name=Sunrise Masks, nodeColor=<rgba 0.1719 1 0 1>}
  NamedRerouteDeclaration(None=Constant() {r=1}) {name=Thick, nodeColor=<rgba 1 0.2109 0 1>}
  NamedRerouteDeclaration(None=Saturate(None=@n43.R)) {name=DepthMask, nodeColor=<rgba 0.2188 0 1 1>}
  NamedRerouteUsage()
  OneMinus(None=@n50)
  Reroute(None=@n74)
  Reroute(None=@n17)
  n75 = ScalarParameter() {defaultValue=1, group=Ice, parameterName="Add Soft Curvature "}
  n76 = TextureSampleParameter2D() {
   group=Ice Fallback
   parameterName=Ice Mask RGB
   samplerType=SAMPLERTYPE_Masks
   sortPriority=2
   texture=/Game/Global/Textures/Default/Blanks/Temp_specular_Mask
   channelNames {a="", b="", g=AO, r=Curvature}
  }
  Saturate(None=Add(A=Multiply(A=@n75, B=@n76.Curvature)))
  ScalarParameter() {defaultValue=55, parameterName=sr1}
  ShadingModel()
 ]}
