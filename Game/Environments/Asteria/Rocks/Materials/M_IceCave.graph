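# Subsurface Nanite material for Asteria ice cave rocks and spline meshes, mixing ice with landscape snow.
# It height-blends ice and snow using ice noise, adding a fallback detail normal for non-Nanite meshes at Epic quality.
# Vertex alpha blends in landscape RVT normals, with switches to skip RVT, before post-RVT snow; Forward and Mobile bypass this.
# A snow mask from vertex alpha and upward normals weights Ice and Snow physical materials.
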Material /Game/Environments/Asteria/Rocks/Materials/M_IceCave
graph {
 properties {
  bEnableNewHLSLGeneratorOverridden=true
  bTangentSpaceNormal=false
  bUseMaterialAttributes=true
  bUsedWithNanite=true
  bUsedWithSplineMeshes=true
  pixelDepthOffsetMode=PDOM_Legacy
  shadingModel=MSM_Subsurface
  subsurfaceProfile=/Game/Environments/Asteria/Rocks/Textures/Crystal_SSP
  usedShadingModels=Subsurface
 }
 nodes [
  n1 = MaterialFunctionCall() {materialFunction=/Game/Athena/Asteria/Environments/Landscape/Materials/MaterialFunctions/MF_AsteriaLandscapeLayers}
  n2 = NamedRerouteUsage()
  n3 = NamedRerouteUsage()
  n4 = NamedRerouteUsage()
  n5 = MaterialFunctionCall() {materialFunction=/Game/Environments/Asteria/Rocks/Materials/MaterialFunctions/MF_IceCave_Ice}
  n6 = Constant() {r=0.5}
  n7 = GetMaterialAttributes(None=BlendMaterialAttributes(
    A=@n5.Result
    B=SetMaterialAttributes!(MaterialAttributes=@n1.Snow, Subsurface Color=@n2, Opacity=@n3) {attributeSetTypes [#1, #2]}
    Alpha=MaterialFunctionCall(Transition Phase=@n4, Height Texture=@n5.Noise, Contrast=@n6) {materialFunction=&HeightLerp}.Alpha
   )) {attributeGetTypes [#3]}
  n8 = TextureSampleParameter2D() {
   desc="Use to support Non Nanite meshes fallbacks. "
   group=Ice Fallback
   parameterName=Fallback Normal
   samplerType=SAMPLERTYPE_Normal
   texture=&FlatNormalBlue
  }
  Add(A=@n7.Normal, B=Multiply(A=@n8.RGB, B=Constant3Vector() {constant=<rgba 1 1 0 0>}))
  n9 = QualitySwitch(
   Default=@n7.Normal
   Epic=Normalize(VectorInput=MaterialFunctionCall(BaseNormal=@n7.Normal, AdditionalNormal=@n8.RGB) {materialFunction=&BlendAngleCorrectedNormals}.Result)
  )
  n10 = GetMaterialAttributes(None=SetMaterialAttributes!(MaterialAttributes=@n7.MaterialAttributes, Normal=NaniteReplace!(Default=@n9, Nanite=@n7.Normal)) {attributeSetTypes [#3]}) {attributeGetTypes [#3]}
  n11 = Transform(None=@n10.Normal)
  NamedRerouteDeclaration(None=SetMaterialAttributes(
    MaterialAttributes=@n10.MaterialAttributes
    Normal=NamedRerouteDeclaration(None=@n11) {name=Material Normals WS, nodeColor=<rgba 0.7344 1 0 1>}
   ) {attributeSetTypes [#3]}) {name=PreRVT, nodeColor=<rgba 0.4766 1 0 1>}
  n12 = NamedRerouteUsage()
  n13 = NamedRerouteUsage()
  n14 = ComponentMask(None=VertexNormalWS()) {b=true}
  n15 = OneMinus(None=@n14)
  n16 = Abs(None=@n14)
  n17 = StaticSwitchParameter(True=Saturate(None=@n15), False=Saturate(None=Multiply(A=@n15, B=@n16))) {group=RVT, parameterName="Skip RVT on Bottom && Sides"}
  n18 = StaticSwitchParameter(
   True=@n12
   False=StaticSwitchParameter(True=BlendMaterialAttributes(A=@n13, B=@n12, Alpha=@n17), False=@n13) {group=RVT, parameterName=Skip RVT on Bottom}
  ) {group=RVT, parameterName=SkipRVT}
  n19 = GetMaterialAttributes(None=MaterialFunctionCall!(Mask=NamedRerouteUsage(), SlipperySnow_Mask, "Material;"=@n18) {materialFunction=/Game/Athena/Asteria/Environments/Landscape/Materials/MaterialFunctions/MF_SnowPostRVT}.SnowMaterial) {attributeGetTypes [#1, #2]}
  NamedRerouteDeclaration(None=@n19.Subsurface Color) {name=SSSColor, nodeColor=<rgba 0.7578 1 0 1>}
  NamedRerouteDeclaration(None=@n19.Opacity) {name=SSSDepth, nodeColor=<rgba 0.3359 1 0 1>}
  n20 = NamedRerouteUsage()
  n21 = RuntimeVirtualTextureSampleParameter!(UVs, World Position, Mip Bias) {
   group=RVT
   materialType=BaseColor_Normal_Specular_YCoCg
   mipValueMode=RVTMVM_MipBias
   parameterName=RVT
   virtualTexture=/Game/Athena/Apollo/Maps/Landscape/VT/RVT_Apollo_Landscape
  }
  n22 = VertexColor()
  NamedRerouteDeclaration(None=BlendMaterialAttributes(
    A=SetMaterialAttributes(MaterialAttributes=@n20, Normal=@n21.Normal) {attributeSetTypes [#3]}
    B=@n20
    Alpha=Saturate(None=@n22.A)
   ) {vertexAttributeBlendType=UseA}) {name=RVT On, nodeColor=<rgba 0 0.8594 1 1>}
  n23 = OneMinus(None=VertexColor().A)
  n24 = ComponentMask(None=SphereMask(
    A=VertexNormalWS()
    B=Constant3Vector() {constant=<rgba 0 0 1 0>}
    Radius=ScalarParameter() {defaultValue=1, parameterName=Up}
    Hardness=ScalarParameter() {defaultValue=0.5, parameterName=UPC}
   )) {b=true}
  n25 = NamedRerouteDeclaration(None=Saturate(None=Add(A=@n23, B=@n24))) {name=SnowMask, nodeColor=<rgba 0 0.2266 1 1>}
  MaterialExpressionLinearInterpolate_3 = LinearInterpolate(Alpha=@n25) {constA=1, constB=0}
  MaterialExpressionLinearInterpolate_0 = LinearInterpolate(Alpha=@n25)
  PhysicalMaterialOutput!(Ice=@MaterialExpressionLinearInterpolate_3, Snow=@MaterialExpressionLinearInterpolate_0) {inputs [
    {
     physicalMaterial=/Game/FortressPhysicalMaterials/Ice
     input {
      expression=&self:MaterialExpressionLinearInterpolate_3
      inputName=None
      mask=0
      maskA=0
      maskB=0
      maskG=0
      maskR=0
      outputIndex=0
     }
    }
    {
     physicalMaterial=/Game/FortressPhysicalMaterials/Snow
     input {
      expression=&self:MaterialExpressionLinearInterpolate_0
      inputName=None
      mask=0
      maskA=0
      maskB=0
      maskG=0
      maskR=0
      outputIndex=0
     }
    }
   ]}
  n26 = NamedRerouteUsage()
  n27 = Saturate(None=NamedRerouteUsage())
  n28 = MaterialFunctionCall(NoRVT=@n26, RVT=MaterialProxyReplace(Realtime=BlendMaterialAttributes(A=@n18, B=@n19.MaterialAttributes, Alpha=@n27), MaterialProxy=@n26)) {materialFunction=&FN_RVT_Switch_Material}
  ShadingPathSwitch(Default=@n28.Result, Deferred=@n28.Result, Forward=@n26, Mobile=@n26)
 ]
}
