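# Landscape material function for the Asteria forest biome, outputting blended Material attributes and Height.
# Forest floor and leaves texture sets share UVs from landscape coordinates or scaled world XY, with each mask's red channel driving both Specular and Roughness.
# The BiomeMask input's green channel (default) blends floor to leaves, including Height from diffuse alphas.
# Its blue channel (default) blends in a Sakura-tinted base color over that result.
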MaterialFunction /Game/Athena/Asteria/Environments/Landscape/Materials/MaterialFunctions/MF_Asteria_Biome_Forest
graph {nodes [
  Constant() {r=1}
  n1 = ComponentMask(None=WorldPosition() {worldPositionShaderOffset=WPT_ExcludeAllShaderOffsets}.XYZ) {g=true, r=true}
  n2 = NamedRerouteUsage()
  n3 = Multiply(
   A=StaticSwitchParameter(True=LandscapeLayerCoords() {mappingScale=10}, False=Divide(A=@n1, B=@n2)) {defaultValue=true, parameterName=UseLandscapeUVs}
   B=ScalarParameter() {defaultValue=2, group=Forest, parameterName=Forest Scale}
  )
  n4 = TextureSampleParameter2D(UVs=@n3) {
   group=Forest
   parameterName=Forest_D2
   samplerSource=SSM_Wrap_WorldGroupSettings
   texture=/Game/Environments/Apollo/Sets/Terrain/Textures/T_Apollo_Forest_Floor_D_01
  }
  n5 = TextureSampleParameter2D(UVs=@n3) {
   parameterName=Forest_S_02
   samplerSource=SSM_Wrap_WorldGroupSettings
   samplerType=SAMPLERTYPE_Masks
   texture=/Game/Environments/Apollo/Sets/Terrain/Textures/T_Apollo_Forest_Floor_S_01
  }
  n6 = TextureSampleParameter2D(UVs=@n3) {
   parameterName=Forest_N_02
   samplerSource=SSM_Wrap_WorldGroupSettings
   samplerType=SAMPLERTYPE_Normal
   texture=/Game/Environments/Apollo/Sets/Terrain/Textures/T_Apollo_Forest_Floor_N_01
  }
  n7 = SetMaterialAttributes!(MaterialAttributes, BaseColor=@n4.RGB, Specular=@n5.R, Roughness=@n5.R, Normal=@n6.RGB) {attributeSetTypes [
    #1
    #2
    #3
    #4
   ]}
  n8 = TextureSampleParameter2D(UVs=@n3) {
   group=Forest
   parameterName=Forest_D
   samplerSource=SSM_Wrap_WorldGroupSettings
   texture=/Game/Environments/Apollo/Sets/Terrain/Textures/T_Apollo_Forest_Leaves_D_01
  }
  n9 = TextureSampleParameter2D(UVs=@n3) {
   group=Forest
   parameterName=Forest_S
   samplerSource=SSM_Wrap_WorldGroupSettings
   samplerType=SAMPLERTYPE_Masks
   texture=/Game/Environments/Apollo/Sets/Terrain/Textures/T_Apollo_Forest_Leaves_S_01
  }
  n10 = TextureSampleParameter2D(UVs=@n3) {
   group=Forest
   parameterName=Forest_N
   samplerSource=SSM_Wrap_WorldGroupSettings
   samplerType=SAMPLERTYPE_Normal
   texture=/Game/Environments/Apollo/Sets/Terrain/Textures/T_Apollo_Forest_Leaves_N_01
  }
  n11 = SetMaterialAttributes(BaseColor=@n8.RGB, Specular=@n9.R, Roughness=@n9.R, Normal=@n10.RGB) {attributeSetTypes [
    #1
    #2
    #3
    #4
   ]}
  n12 = FunctionInput(Preview=Constant()) {bUsePreviewValueAsDefault=true, inputName=BiomeMask-RGBA, inputType=FunctionInput_Vector4}
  n13 = ChannelMaskParameter(None=@n12) {maskChannel=Green, parameterName=ForestFloor-BiomeMaskChannel, defaultValue=<rgba 0 1 0 0>}
  n14 = BlendMaterialAttributes(A=@n7, B=@n11, Alpha=@n13.R)
  n15 = GetMaterialAttributes(None=@n14) {attributeGetTypes [#1]}
  n16 = Multiply(A=@n15.BaseColor, B=VectorParameter() {parameterName=Sakura_ForestFloorTint, defaultValue=<rgba 0.2991 0.6473 1.1 1>}.RGB)
  FunctionOutput(None=BlendMaterialAttributes(
    A=@n14
    B=SetMaterialAttributes!(MaterialAttributes=@n15.MaterialAttributes, BaseColor=@n16) {attributeSetTypes [#1]}
    Alpha=ChannelMaskParameter(None=@n12) {maskChannel=Blue, parameterName=Sakura-BiomeMaskChannel, defaultValue=<rgba 0 0 1 0>}.R
   )) {outputName=Material}
  FunctionOutput(None=LinearInterpolate(A=@n4.A, B=@n8.A, Alpha=@n13.R)) {outputName=Height, sortPriority=1}
  NamedRerouteDeclaration(None=ScalarParameter() {defaultValue=600, parameterName=Landscape_BaseScale}) {name=LandscapeCoordinate_BaseScale, nodeColor=<rgba 1 0 0.4688 1>}
  TextureCoordinate() {uTiling=8, vTiling=8}
 ]}
