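# Material function for Asteria rock materials that blends material B into subsurface material A by a saturated alpha.
# B's subsurface color is zeroed, while A's base and subsurface colors are pre-lerped toward B's base color as alpha falls.
# That subsurface tint is scaled by alpha remapped through the Min/Max inputs, confining it to a narrow band by default.
# The final attribute blend takes vertex attributes from the subsurface side.
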MaterialFunction /Game/Environments/Asteria/Rocks/Materials/MaterialFunctions/MF_BlendSubsurfaceMaterial
graph {nodes [
  n1 = GetMaterialAttributes(None=FunctionInput(Preview=MakeMaterialAttributes!(
     BaseColor
     Metallic
     Specular
     Roughness
     Anisotropy
     EmissiveColor
     Opacity
     OpacityMask
     Normal
     Tangent
     WorldPositionOffset
     SubsurfaceColor
     ClearCoat
     ClearCoatRoughness
     AmbientOcclusion
     Refraction
     CustomizedUVs_0
     CustomizedUVs_1
     CustomizedUVs_2
     CustomizedUVs_3
     CustomizedUVs_4
     CustomizedUVs_5
     CustomizedUVs_6
     CustomizedUVs_7
     PixelDepthOffset
     ShadingModel
     Displacement
    )) {bUsePreviewValueAsDefault=true, inputName=B, inputType=FunctionInput_MaterialAttributes}) {attributeGetTypes [#1]}
  n2 = Constant()
  n3 = GetMaterialAttributes(None=FunctionInput(Preview=MakeMaterialAttributes()) {bUsePreviewValueAsDefault=true, inputName="A (Subsurface)", inputType=FunctionInput_MaterialAttributes}) {attributeGetTypes [#2, #3, #1]}
  n4 = Saturate(None=FunctionInput() {bUsePreviewValueAsDefault=true, inputName=Alpha})
  n5 = FunctionInput(Preview=Constant() {r=0.85}) {bUsePreviewValueAsDefault=true, inputName=Subsurface Blend Min, inputType=FunctionInput_Scalar}
  n6 = FunctionInput(Preview=Constant() {r=1}) {bUsePreviewValueAsDefault=true, inputName=Subsurface Blend Max, inputType=FunctionInput_Scalar}
  n7 = Saturate(None=OneMinus(None=@n4))
  n8 = LinearInterpolate(
   A=@n3.SubsurfaceColor
   B=Multiply(
    A=@n1.BaseColor
    B=MaterialFunctionCall(Input=@n4, Input Low=@n5, Input High=@n6) {materialFunction=&RemapValueRange}.Result
   )
   Alpha=@n7
  )
  n9 = LinearInterpolate(A=@n3.BaseColor, B=@n1.BaseColor, Alpha=@n7) {constB=0}
  FunctionOutput(None=BlendMaterialAttributes(
    A=SetMaterialAttributes!(MaterialAttributes=@n1.MaterialAttributes, SubsurfaceColor=@n2) {attributeSetTypes [#2]}
    B=SetMaterialAttributes!(MaterialAttributes=@n3.MaterialAttributes, SubsurfaceColor=@n8, Opacity, BaseColor=@n9) {attributeSetTypes [#2, #3, #1]}
    Alpha=@n4
   ) {vertexAttributeBlendType=UseB}) {outputName=Result MA}
  MaterialFunctionCall() {materialFunction=&RemapValue_0-1}
 ]}
