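# Base material function for Asteria rocks, outputting layered surface attributes plus AO, up-side mask, detail mask, dust and merged heights.
# World-aligned texture sets supply color, spec and normals, with desaturation and tint controls.
# Masks from vertex normals, world height and local bounds drive height-lerped attribute blends between layers.
# Distance field AO, quality and static switches, and an RVT switch for a scalar input add optional variation.
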MaterialFunction /Game/Environments/Asteria/Rocks/Materials/MaterialFunctions/MF_Rock2022Base
graph {expressions {
  MaterialExpressionAbs_0 {class=MaterialExpressionAbs, inputs [None {from=MaterialExpressionComponentMask_2, output=""}], properties {}}
  MaterialExpressionAbs_1 {class=MaterialExpressionAbs, inputs [None {from=MaterialExpressionDotProduct_1, output=""}], properties {}}
  MaterialExpressionAbs_2 {class=MaterialExpressionAbs, inputs [None {from=MaterialExpressionMultiply_64, output=""}], properties {}}
  MaterialExpressionAbs_4 {class=MaterialExpressionAbs, inputs [None {from=MaterialExpressionComponentMask_15, output=""}], properties {}}
  MaterialExpressionActorPositionWS_0 {class=MaterialExpressionActorPositionWS, inputs [], properties {}}
  MaterialExpressionAdd_0 {
   class=MaterialExpressionAdd
   inputs [
    A {from=MaterialExpressionComponentMask_2}
    B {from=MaterialExpressionConstant_14}
   ]
  }
  MaterialExpressionAdd_1 {
   class=MaterialExpressionAdd
   inputs [
    A {from=MaterialExpressionSaturate_8}
    B {from=MaterialExpressionSaturate_5}
   ]
  }
  MaterialExpressionAdd_2 {
   class=MaterialExpressionAdd
   inputs [
    A {from=MaterialExpressionGetMaterialAttributes_5, output=Normal}
    B {from=MaterialExpressionNamedRerouteUsage_45}
   ]
  }
  MaterialExpressionAdd_3 {
   class=MaterialExpressionAdd
   inputs [
    A {from=MaterialExpressionVectorParameter_5, output=RGB}
    B {from=MaterialExpressionActorPositionWS_0}
   ]
  }
  MaterialExpressionAdd_4 {
   class=MaterialExpressionAdd
   inputs [
    A {from=MaterialExpressionMaterialFunctionCall_3, output=Local Bounds Minimum}
    B {from=MaterialExpressionScalarParameter_1}
   ]
  }
  MaterialExpressionAdd_5 {
   class=MaterialExpressionAdd
   inputs [
    A {from=MaterialExpressionMultiply_33}
    B {from=MaterialExpressionScalarParameter_39}
   ]
   properties {constB=-0.75}
  }
  MaterialExpressionAppendVector_0 {
   class=MaterialExpressionAppendVector
   inputs [
    A {from=MaterialExpressionMultiply_7, output=""}
    B {from=MaterialExpressionComponentMask_1, output=""}
   ]
  }
  MaterialExpressionAppendVector_1 {
   class=MaterialExpressionAppendVector
   inputs [
    A {from=MaterialExpressionMultiply_24, output=""}
    B {from=MaterialExpressionMaterialFunctionCall_16, output=B}
   ]
  }
  MaterialExpressionAppendVector_2 {
   class=MaterialExpressionAppendVector
   inputs [
    A {from=MaterialExpressionMaterialFunctionCall_16, output=R}
    B {from=MaterialExpressionMaterialFunctionCall_16, output=R}
   ]
  }
  MaterialExpressionAppendVector_3 {
   class=MaterialExpressionAppendVector
   inputs [
    A {from=MaterialExpressionMultiply_38, output=""}
    B {from=MaterialExpressionMaterialFunctionCall_19, output=B}
   ]
  }
  MaterialExpressionAppendVector_6 {
   class=MaterialExpressionAppendVector
   inputs [
    A {from=MaterialExpressionMaterialFunctionCall_13, output=R}
    B {from=MaterialExpressionMaterialFunctionCall_13, output=R}
   ]
  }
  MaterialExpressionAppendVector_7 {
   class=MaterialExpressionAppendVector
   inputs [
    A {from=MaterialExpressionMultiply_22, output=""}
    B {from=MaterialExpressionMaterialFunctionCall_13, output=B}
   ]
  }
  MaterialExpressionAppendVector_8 {
   class=MaterialExpressionAppendVector
   inputs [
    A {from=MaterialExpressionMaterialFunctionCall_19, output=R}
    B {from=MaterialExpressionMaterialFunctionCall_19, output=R}
   ]
  }
  MaterialExpressionBlendMaterialAttributes_0 {
   class=MaterialExpressionBlendMaterialAttributes
   inputs [
    A {from=MaterialExpressionSetMaterialAttributes_0, output=""}
    B {from=MaterialExpressionGetMaterialAttributes_2, output=MaterialAttributes}
    Alpha {from=MaterialExpressionSaturate_3, output=""}
   ]
   properties {}
  }
  MaterialExpressionBlendMaterialAttributes_1 {
   class=MaterialExpressionBlendMaterialAttributes
   inputs [
    A {from=MaterialExpressionSetMaterialAttributes_1, output=""}
    B {from=MaterialExpressionNamedRerouteUsage_0, output=""}
    Alpha {from=MaterialExpressionMultiply_2, output=""}
   ]
   properties {}
  }
  MaterialExpressionBlendMaterialAttributes_10 {
   class=MaterialExpressionBlendMaterialAttributes
   inputs [
    A {from=MaterialExpressionGetMaterialAttributes_1, output=MaterialAttributes}
    B {from=MaterialExpressionSetMaterialAttributes_15, output=""}
    Alpha {from=MaterialExpressionSaturate_22, output=""}
   ]
   properties {}
  }
  MaterialExpressionBlendMaterialAttributes_2 {
   class=MaterialExpressionBlendMaterialAttributes
   inputs [
    A {from=MaterialExpressionReroute_1, output=""}
    B {from=MaterialExpressionSetMaterialAttributes_3, output=""}
    Alpha {from=MaterialExpressionMaterialFunctionCall_4, output=Result}
   ]
   properties {}
  }
  MaterialExpressionBlendMaterialAttributes_3 {
   class=MaterialExpressionBlendMaterialAttributes
   inputs [
    A {from=MaterialExpressionSetMaterialAttributes_4, output=""}
    B {from=MaterialExpressionNamedRerouteDeclaration_17, output=""}
    Alpha {from=MaterialExpressionNamedRerouteUsage_28, output=""}
   ]
   properties {}
  }
  MaterialExpressionBlendMaterialAttributes_4 {
   class=MaterialExpressionBlendMaterialAttributes
   inputs [
    A {from=MaterialExpressionNamedRerouteUsage_5, output=""}
    B {from=MaterialExpressionBlendMaterialAttributes_0, output=""}
    Alpha {from=MaterialExpressionStaticSwitchParameter_4, output=""}
   ]
   properties {}
  }
  MaterialExpressionBlendMaterialAttributes_6 {
   class=MaterialExpressionBlendMaterialAttributes
   inputs [
    A {from=MaterialExpressionNamedRerouteUsage_30, output=""}
    B {from=MaterialExpressionNamedRerouteUsage_57, output=""}
    Alpha {from=MaterialExpressionNamedRerouteDeclaration_23, output=""}
   ]
   properties {}
  }
  MaterialExpressionBlendMaterialAttributes_7 {
   class=MaterialExpressionBlendMaterialAttributes
   inputs [
    A {from=MaterialExpressionNamedRerouteUsage_52, output=""}
    B {from=MaterialExpressionSetMaterialAttributes_12, output=""}
    Alpha {from=MaterialExpressionSaturate_14, output=""}
   ]
   properties {}
  }
  MaterialExpressionComponentMask_0 {class=MaterialExpressionComponentMask, inputs [None {from=MaterialExpressionReroute_17, output=Input}], properties {g=true, r=true}}
  MaterialExpressionComponentMask_1 {class=MaterialExpressionComponentMask, inputs [None {from=MaterialExpressionReroute_20, output=Input2}], properties {b=true}}
  MaterialExpressionComponentMask_11 {class=MaterialExpressionComponentMask, inputs [None {from=MaterialExpressionNamedRerouteUsage_39}], properties {b=true}}
  MaterialExpressionComponentMask_12 {class=MaterialExpressionComponentMask, inputs [None {from=MaterialExpressionReroute_23}], properties {b=true}}
  MaterialExpressionComponentMask_13 {
   class=MaterialExpressionComponentMask
   inputs [None {from=MaterialExpressionMaterialFunctionCall_11, output=Z Diffuse Texture}]
   properties {a=true}
  }
  MaterialExpressionComponentMask_14 {class=MaterialExpressionComponentMask, inputs [None {from=MaterialExpressionDivide_7}], properties {b=true}}
  MaterialExpressionComponentMask_15 {class=MaterialExpressionComponentMask, inputs [None {from=MaterialExpressionVertexNormalWS_3}], properties {b=true}}
  MaterialExpressionComponentMask_16 {
   class=MaterialExpressionComponentMask
   inputs [None {from=MaterialExpressionMaterialFunctionCall_11, output=Z Diffuse Texture}]
   properties {b=true, g=true, r=true}
  }
  MaterialExpressionComponentMask_18 {class=MaterialExpressionComponentMask, inputs [None {from=MaterialExpressionNamedRerouteDeclaration_16}], properties {b=true}}
  MaterialExpressionComponentMask_19 {class=MaterialExpressionComponentMask, inputs [None {from=MaterialExpressionNamedRerouteDeclaration_4}], properties {b=true, g=true, r=true}}
  MaterialExpressionComponentMask_2 {class=MaterialExpressionComponentMask, inputs [None {from=MaterialExpressionSubtract_0}], properties {b=true}}
  MaterialExpressionComponentMask_28 {class=MaterialExpressionComponentMask, inputs [None {from=MaterialExpressionDivide_1}], properties {g=true, r=true}}
  MaterialExpressionComponentMask_3 {class=MaterialExpressionComponentMask, inputs [None {from=MaterialExpressionNamedRerouteUsage_4}], properties {b=true}}
  MaterialExpressionComponentMask_30 {class=MaterialExpressionComponentMask, inputs [None {from=MaterialExpressionVertexNormalWS_11}], properties {b=true}}
  MaterialExpressionComponentMask_4 {class=MaterialExpressionComponentMask, inputs [None {from=MaterialExpressionStaticSwitchParameter_9}], properties {b=true, g=true, r=true}}
  MaterialExpressionComponentMask_6 {class=MaterialExpressionComponentMask, inputs [None {from=MaterialExpressionStaticSwitchParameter_9}], properties {a=true}}
  MaterialExpressionComponentMask_8 {class=MaterialExpressionComponentMask, inputs [None {from=MaterialExpressionNamedRerouteDeclaration_4}], properties {a=true}}
  MaterialExpressionComponentMask_9 {class=MaterialExpressionComponentMask, inputs [None {from=MaterialExpressionVertexNormalWS_1}], properties {b=true}}
  MaterialExpressionConstant3Vector_0 {class=MaterialExpressionConstant3Vector, properties {constant=<rgba 0 0 1 1>}}
  MaterialExpressionConstant3Vector_1 {class=MaterialExpressionConstant3Vector, properties {constant=<rgba 0 0 -1 0>}}
  MaterialExpressionConstant3Vector_10 {class=MaterialExpressionConstant3Vector, properties {constant=<rgba 3 3 1 0>}}
  MaterialExpressionConstant3Vector_12 {class=MaterialExpressionConstant3Vector, properties {constant=<rgba 0 0 1 0>}}
  MaterialExpressionConstant3Vector_16 {class=MaterialExpressionConstant3Vector, properties {constant=<rgba 0 0 -1 0>}}
  MaterialExpressionConstant3Vector_3 {class=MaterialExpressionConstant3Vector, properties {constant=<rgba 1.5 1.5 1 0>}}
  MaterialExpressionConstant3Vector_4 {class=MaterialExpressionConstant3Vector, properties {constant=<rgba 0 0 1 0>}}
  MaterialExpressionConstant3Vector_5 {class=MaterialExpressionConstant3Vector, properties {constant=<rgba 0 0 1 0>}}
  MaterialExpressionConstant3Vector_6 {class=MaterialExpressionConstant3Vector, properties {constant=<rgba 1 1 0 0>}}
  MaterialExpressionConstant3Vector_7 {class=MaterialExpressionConstant3Vector, properties {constant=<rgba 0 0 -1 0>}}
  MaterialExpressionConstant_0 {class=MaterialExpressionConstant, properties {r=0.25}}
  MaterialExpressionConstant_1 {class=MaterialExpressionConstant, properties {r=1}}
  MaterialExpressionConstant_10 {class=MaterialExpressionConstant, properties {r=0.85}}
  MaterialExpressionConstant_11 {class=MaterialExpressionConstant, properties {r=0.05}}
  MaterialExpressionConstant_12 {class=MaterialExpressionConstant, properties {r=1}}
  MaterialExpressionConstant_13 {class=MaterialExpressionConstant, properties {r=30}}
  MaterialExpressionConstant_14 {class=MaterialExpressionConstant, properties {r=15}}
  MaterialExpressionConstant_15 {class=MaterialExpressionConstant, properties {r=0.35}}
  MaterialExpressionConstant_16 {class=MaterialExpressionConstant, properties {r=0.5}}
  MaterialExpressionConstant_17 {class=MaterialExpressionConstant, properties {r=0.7}}
  MaterialExpressionConstant_18 {class=MaterialExpressionConstant, properties {r=0.5}}
  MaterialExpressionConstant_19 {class=MaterialExpressionConstant, properties {r=2}}
  MaterialExpressionConstant_2 {class=MaterialExpressionConstant, properties {r=0.15}}
  MaterialExpressionConstant_20 {class=MaterialExpressionConstant}
  MaterialExpressionConstant_21 {class=MaterialExpressionConstant, properties {r=1}}
  MaterialExpressionConstant_22 {class=MaterialExpressionConstant, properties {r=1}}
  MaterialExpressionConstant_23 {class=MaterialExpressionConstant}
  MaterialExpressionConstant_24 {class=MaterialExpressionConstant}
  MaterialExpressionConstant_28 {class=MaterialExpressionConstant, properties {r=0.2}}
  MaterialExpressionConstant_3 {class=MaterialExpressionConstant, properties {r=2}}
  MaterialExpressionConstant_30 {class=MaterialExpressionConstant, properties {r=1.3436}}
  MaterialExpressionConstant_31 {class=MaterialExpressionConstant, properties {r=0.2613}}
  MaterialExpressionConstant_32 {class=MaterialExpressionConstant, properties {r=0.1297}}
  MaterialExpressionConstant_33 {class=MaterialExpressionConstant, properties {r=0.7982}}
  MaterialExpressionConstant_34 {class=MaterialExpressionConstant, properties {r=0.376}}
  MaterialExpressionConstant_35 {class=MaterialExpressionConstant, properties {r=-0.032}}
  MaterialExpressionConstant_36 {class=MaterialExpressionConstant, properties {r=-0.1161}}
  MaterialExpressionConstant_37 {class=MaterialExpressionConstant, properties {r=0.6293}}
  MaterialExpressionConstant_39 {class=MaterialExpressionConstant, properties {r=0.75}}
  MaterialExpressionConstant_4 {class=MaterialExpressionConstant, properties {r=0.75}}
  MaterialExpressionConstant_45 {class=MaterialExpressionConstant, properties {r=150}}
  MaterialExpressionConstant_46 {class=MaterialExpressionConstant, properties {r=3}}
  MaterialExpressionConstant_47 {class=MaterialExpressionConstant, properties {r=-1}}
  MaterialExpressionConstant_6 {class=MaterialExpressionConstant, properties {r=3}}
  MaterialExpressionConstant_7 {class=MaterialExpressionConstant, properties {r=-2}}
  MaterialExpressionConstant_8 {class=MaterialExpressionConstant, properties {r=4}}
  MaterialExpressionConstant_9 {class=MaterialExpressionConstant, properties {r=4}}
  MaterialExpressionDesaturation_0 {
   class=MaterialExpressionDesaturation
   inputs [
    None {from=MaterialExpressionComponentMask_16, output=""}
    Fraction {from=MaterialExpressionScalarParameter_6, output=""}
   ]
   properties {}
  }
  MaterialExpressionDesaturation_1 {
   class=MaterialExpressionDesaturation
   inputs [
    None {from=MaterialExpressionComponentMask_19, output=""}
    Fraction {from=MaterialExpressionScalarParameter_8, output=""}
   ]
   properties {}
  }
  MaterialExpressionDesaturation_2 {
   class=MaterialExpressionDesaturation
   inputs [
    None {from=MaterialExpressionComponentMask_4, output=""}
    Fraction {from=MaterialExpressionScalarParameter_14, output=""}
   ]
   properties {}
  }
  MaterialExpressionDesaturation_3 {
   class=MaterialExpressionDesaturation
   inputs [
    None {from=MaterialExpressionSaturate_17, output=""}
    Fraction {from=MaterialExpressionScalarParameter_37, output=""}
   ]
   properties {}
  }
  MaterialExpressionDistanceFieldApproxAO_1 {
   class=MaterialExpressionDistanceFieldApproxAO
   inputs [
    World Position {from="", output=""}
    Normal {from="", output=""}
    BaseDistance {from=MaterialExpressionScalarParameter_20, output=""}
    Radius {from=MaterialExpressionScalarParameter_19, output=""}
   ]
   properties {numSteps=3}
  }
  MaterialExpressionDivide_0 {
   class=MaterialExpressionDivide
   inputs [
    A {from=MaterialExpressionMaterialFunctionCall_17, output=Result}
    B {from=MaterialExpressionScalarParameter_45}
   ]
  }
  MaterialExpressionDivide_1 {
   class=MaterialExpressionDivide
   inputs [
    A {from=MaterialExpressionWorldPosition_1, output=XYZ}
    B {from=MaterialExpressionMultiply_62}
   ]
  }
  MaterialExpressionDivide_2 {
   class=MaterialExpressionDivide
   inputs [
    A {from=MaterialExpressionAbs_0}
    B {from=MaterialExpressionLinearInterpolate_9}
   ]
  }
  MaterialExpressionDivide_3 {
   class=MaterialExpressionDivide
   inputs [
    A {from=MaterialExpressionComponentMask_2}
    B {from=MaterialExpressionMultiply_3}
   ]
  }
  MaterialExpressionDivide_4 {
   class=MaterialExpressionDivide
   inputs [
    A {from=MaterialExpressionAdd_0}
    B {from=MaterialExpressionConstant_13}
   ]
  }
  MaterialExpressionDivide_6 {
   class=MaterialExpressionDivide
   inputs [
    A {from=MaterialExpressionNamedRerouteUsage_34}
    B {from=MaterialExpressionScalarParameter_30}
   ]
  }
  MaterialExpressionDivide_7 {
   class=MaterialExpressionDivide
   inputs [
    A {from=MaterialExpressionSubtract_6}
    B {from=MaterialExpressionMaterialFunctionCall_3, output=Local Bounds Size}
   ]
  }
  MaterialExpressionDotProduct_0 {
   class=MaterialExpressionDotProduct
   inputs [
    A {from=MaterialExpressionVertexNormalWS_10, output=""}
    B {from=MaterialExpressionConstant3Vector_16, output=""}
   ]
   properties {}
  }
  MaterialExpressionDotProduct_1 {
   class=MaterialExpressionDotProduct
   inputs [
    A {from=MaterialExpressionVertexNormalWS_0, output=""}
    B {from=MaterialExpressionConstant3Vector_1, output=""}
   ]
   properties {}
  }
  MaterialExpressionDotProduct_2 {
   class=MaterialExpressionDotProduct
   inputs [
    A {from=MaterialExpressionVertexNormalWS_5, output=""}
    B {from=MaterialExpressionConstant3Vector_7, output=""}
   ]
   properties {}
  }
  MaterialExpressionFunctionInput_0 {class=MaterialExpressionFunctionInput, inputs [Preview {from="", output=""}], properties {inputName=RVTH, inputType=FunctionInput_Scalar}}
  MaterialExpressionFunctionInput_2 {
   class=MaterialExpressionFunctionInput
   inputs [Preview {from=MaterialExpressionStaticBool_0, output=""}]
   properties {inputName="Disable Top Texures ( Use mat layer)", inputType=FunctionInput_StaticBool}
  }
  MaterialExpressionFunctionOutput_0 {class=MaterialExpressionFunctionOutput, inputs [None {from=MaterialExpressionStaticSwitchParameter_0, output=""}]}
  MaterialExpressionFunctionOutput_1 {class=MaterialExpressionFunctionOutput, inputs [None {from=MaterialExpressionQualitySwitch_1, output=""}], properties {outputName=AO}}
  MaterialExpressionFunctionOutput_2 {
   class=MaterialExpressionFunctionOutput
   inputs [None {from=MaterialExpressionSaturate_8, output=""}]
   properties {outputName=Up-Side MaskBetter}
  }
  MaterialExpressionFunctionOutput_3 {
   class=MaterialExpressionFunctionOutput
   inputs [None {from=MaterialExpressionNamedRerouteDeclaration_16, output=""}]
   properties {outputName=Details XYZ RGB Mask}
  }
  MaterialExpressionFunctionOutput_4 {class=MaterialExpressionFunctionOutput, inputs [None {from=MaterialExpressionVectorParameter_16, output=""}], properties {outputName=Dust}}
  MaterialExpressionFunctionOutput_5 {
   class=MaterialExpressionFunctionOutput
   inputs [None {from=MaterialExpressionLinearInterpolate_14, output=""}]
   properties {outputName=Height Maps Merged}
  }
  MaterialExpressionGetMaterialAttributes_0 {
   class=MaterialExpressionGetMaterialAttributes
   inputs [None {from=MaterialExpressionStaticSwitchParameter_12, output=""}]
   properties {attributeGetTypes [#1, #2, #3]}
  }
  MaterialExpressionGetMaterialAttributes_1 {
   class=MaterialExpressionGetMaterialAttributes
   inputs [None {from=MaterialExpressionStaticSwitch_0, output=""}]
   properties {attributeGetTypes [#1]}
  }
  MaterialExpressionGetMaterialAttributes_2 {
   class=MaterialExpressionGetMaterialAttributes
   inputs [None {from=MaterialExpressionBlendMaterialAttributes_1, output=""}]
   properties {attributeGetTypes [#2, #3]}
  }
  MaterialExpressionGetMaterialAttributes_3 {
   class=MaterialExpressionGetMaterialAttributes
   inputs [None {from=MaterialExpressionNamedRerouteDeclaration_15, output=""}]
   properties {attributeGetTypes [#3, #4]}
  }
  MaterialExpressionGetMaterialAttributes_5 {
   class=MaterialExpressionGetMaterialAttributes
   inputs [None {from=MaterialExpressionReroute_21, output=""}]
   properties {attributeGetTypes [#2, #1]}
  }
  MaterialExpressionGetMaterialAttributes_7 {
   class=MaterialExpressionGetMaterialAttributes
   inputs [None {from=MaterialExpressionReroute_25, output=""}]
   properties {attributeGetTypes [
     #3
     #1
     #4
     #5
    ]}
  }
  MaterialExpressionLinearInterpolate_0 {
   class=MaterialExpressionLinearInterpolate
   inputs [
    A {from=MaterialExpressionConstant_0}
    B {from=MaterialExpressionGetMaterialAttributes_3, output=Roughness}
    Alpha {from=MaterialExpressionSaturate_1}
   ]
  }
  MaterialExpressionLinearInterpolate_1 {
   class=MaterialExpressionLinearInterpolate
   inputs [
    A {from=MaterialExpressionMultiply_0}
    B {from=MaterialExpressionLinearInterpolate_13}
    Alpha {from=MaterialExpressionNamedRerouteUsage_2}
   ]
  }
  MaterialExpressionLinearInterpolate_10 {
   class=MaterialExpressionLinearInterpolate
   inputs [
    A {from=MaterialExpressionConstant3Vector_3}
    B {from=MaterialExpressionConstant3Vector_10}
    Alpha {from=MaterialExpressionReroute_19, output=Alpha2}
   ]
  }
  MaterialExpressionLinearInterpolate_11 {
   class=MaterialExpressionLinearInterpolate
   inputs [
    A {from=MaterialExpressionConstant3Vector_10}
    B {from=MaterialExpressionLinearInterpolate_10}
    Alpha {from=MaterialExpressionReroute_18, output=Alpha}
   ]
  }
  MaterialExpressionLinearInterpolate_12 {
   class=MaterialExpressionLinearInterpolate
   inputs [
    A {from=MaterialExpressionVectorParameter_3}
    B {from=MaterialExpressionVectorParameter_2}
    Alpha {from=MaterialExpressionNamedRerouteUsage_13}
   ]
  }
  MaterialExpressionLinearInterpolate_13 {
   class=MaterialExpressionLinearInterpolate
   inputs [
    A {from=MaterialExpressionScalarParameter_22}
    B {from=MaterialExpressionScalarParameter_26}
    Alpha {from=MaterialExpressionNamedRerouteUsage_42}
   ]
  }
  MaterialExpressionLinearInterpolate_14 {
   class=MaterialExpressionLinearInterpolate
   inputs [
    A {from=MaterialExpressionNamedRerouteUsage_19}
    B {from=MaterialExpressionNamedRerouteUsage_18}
    Alpha {from=MaterialExpressionNamedRerouteDeclaration_23}
   ]
  }
  MaterialExpressionLinearInterpolate_15 {
   class=MaterialExpressionLinearInterpolate
   inputs [
    A {from=MaterialExpressionConstant_37}
    B {from=MaterialExpressionConstant_36}
    Alpha {from=MaterialExpressionReroute_13, output=Alpha3}
   ]
  }
  MaterialExpressionLinearInterpolate_16 {
   class=MaterialExpressionLinearInterpolate
   inputs [
    A {from=MaterialExpressionGetMaterialAttributes_0, output=Normal}
    B {from=MaterialExpressionConstant3Vector_5}
    Alpha {from=""}
   ]
   properties {constAlpha=0.75}
  }
  MaterialExpressionLinearInterpolate_17 {
   class=MaterialExpressionLinearInterpolate
   inputs [
    A {from=MaterialExpressionGetMaterialAttributes_0, output=BaseColor}
    B {from=MaterialExpressionVectorParameter_4, output=RGB}
    Alpha {from=""}
   ]
   properties {constAlpha=0.75}
  }
  MaterialExpressionLinearInterpolate_18 {
   class=MaterialExpressionLinearInterpolate
   inputs [
    A {from=""}
    B {from=MaterialExpressionScalarParameter_35}
    Alpha {from=MaterialExpressionNamedRerouteUsage_38}
   ]
   properties {constA=1}
  }
  MaterialExpressionLinearInterpolate_19 {
   class=MaterialExpressionLinearInterpolate
   inputs [
    A {from=MaterialExpressionConstant_19}
    B {from=MaterialExpressionConstant_23}
    Alpha {from=MaterialExpressionComponentMask_11}
   ]
  }
  MaterialExpressionLinearInterpolate_20 {
   class=MaterialExpressionLinearInterpolate
   inputs [
    A {from=MaterialExpressionNamedRerouteUsage_17}
    B {from=MaterialExpressionNamedRerouteUsage_16}
    Alpha {from=MaterialExpressionNamedRerouteUsage_28}
   ]
  }
  MaterialExpressionLinearInterpolate_21 {
   class=MaterialExpressionLinearInterpolate
   inputs [
    A {from=MaterialExpressionMultiply_21}
    B {from=MaterialExpressionScalarParameter_16}
    Alpha {from=MaterialExpressionNamedRerouteUsage_22}
   ]
  }
  MaterialExpressionLinearInterpolate_22 {
   class=MaterialExpressionLinearInterpolate
   inputs [
    A {from=MaterialExpressionScalarParameter_28}
    B {from=MaterialExpressionScalarParameter_29}
    Alpha {from=MaterialExpressionStaticSwitchParameter_26}
   ]
  }
  MaterialExpressionLinearInterpolate_23 {
   class=MaterialExpressionLinearInterpolate
   inputs [
    A {from=MaterialExpressionVectorParameter_17, output=Roughness Tweak}
    B {from=MaterialExpressionVectorParameter_7, output=Roughness Tweak}
    Alpha {from=MaterialExpressionSaturate_13}
   ]
  }
  MaterialExpressionLinearInterpolate_24 {
   class=MaterialExpressionLinearInterpolate
   inputs [
    A {from=MaterialExpressionDesaturation_3}
    B {from=MaterialExpressionTextureSampleParameter2D_0, output=RGB}
    Alpha {from=MaterialExpressionSaturate_6}
   ]
  }
  MaterialExpressionLinearInterpolate_25 {
   class=MaterialExpressionLinearInterpolate
   inputs [
    A {from=MaterialExpressionLinearInterpolate_34}
    B {from=""}
    Alpha {from=MaterialExpressionGetMaterialAttributes_7, output=Metallic}
   ]
  }
  MaterialExpressionLinearInterpolate_26 {
   class=MaterialExpressionLinearInterpolate
   inputs [
    A {from=MaterialExpressionLinearInterpolate_8}
    B {from=MaterialExpressionLinearInterpolate_27}
    Alpha {from=MaterialExpressionReroute_0, output=Alpha}
   ]
  }
  MaterialExpressionLinearInterpolate_27 {
   class=MaterialExpressionLinearInterpolate
   inputs [
    A {from=MaterialExpressionConstant_32}
    B {from=MaterialExpressionConstant_33}
    Alpha {from=MaterialExpressionReroute_9, output=Alpha3}
   ]
  }
  MaterialExpressionLinearInterpolate_28 {
   class=MaterialExpressionLinearInterpolate
   inputs [
    A {from=MaterialExpressionConstant_35}
    B {from=MaterialExpressionConstant_34}
    Alpha {from=MaterialExpressionReroute_12, output=Alpha2}
   ]
  }
  MaterialExpressionLinearInterpolate_29 {
   class=MaterialExpressionLinearInterpolate
   inputs [
    A {from=MaterialExpressionLinearInterpolate_15}
    B {from=MaterialExpressionLinearInterpolate_28}
    Alpha {from=MaterialExpressionReroute_10, output=Alpha}
   ]
  }
  MaterialExpressionLinearInterpolate_3 {
   class=MaterialExpressionLinearInterpolate
   inputs [
    A {from=MaterialExpressionMultiply_1}
    B {from=MaterialExpressionScalarParameter_23}
    Alpha {from=MaterialExpressionNamedRerouteUsage_3}
   ]
  }
  MaterialExpressionLinearInterpolate_34 {
   class=MaterialExpressionLinearInterpolate
   inputs [
    A {from=MaterialExpressionVectorParameter_17}
    B {from=MaterialExpressionVectorParameter_7}
    Alpha {from=MaterialExpressionSaturate_13}
   ]
  }
  MaterialExpressionLinearInterpolate_36 {
   class=MaterialExpressionLinearInterpolate
   inputs [
    A {from=MaterialExpressionConstant_47}
    B {from=MaterialExpressionConstant_46}
    Alpha {from=MaterialExpressionNamedRerouteUsage_74}
   ]
  }
  MaterialExpressionLinearInterpolate_4 {
   class=MaterialExpressionLinearInterpolate
   inputs [
    A {from=MaterialExpressionConstant_7}
    B {from=MaterialExpressionConstant_8}
    Alpha {from=MaterialExpressionComponentMask_3}
   ]
  }
  MaterialExpressionLinearInterpolate_43 {
   class=MaterialExpressionLinearInterpolate
   inputs [
    A {from=MaterialExpressionComponentMask_18}
    B {from=MaterialExpressionMin_7}
    Alpha {from=MaterialExpressionSaturate_21}
   ]
  }
  MaterialExpressionLinearInterpolate_5 {
   class=MaterialExpressionLinearInterpolate
   inputs [
    A {from=MaterialExpressionConstant3Vector_4}
    B {from=MaterialExpressionGetMaterialAttributes_2, output=Normal}
    Alpha {from=MaterialExpressionConstant_15}
   ]
  }
  MaterialExpressionLinearInterpolate_6 {
   class=MaterialExpressionLinearInterpolate
   inputs [
    A {from=MaterialExpressionVectorParameter_0}
    B {from=MaterialExpressionMultiply_5}
    Alpha {from=MaterialExpressionVectorParameter_0}
   ]
  }
  MaterialExpressionLinearInterpolate_7 {
   class=MaterialExpressionLinearInterpolate
   inputs [
    A {from=MaterialExpressionVectorParameter_1}
    B {from=MaterialExpressionMultiply_12}
    Alpha {from=MaterialExpressionVectorParameter_1}
   ]
  }
  MaterialExpressionLinearInterpolate_8 {
   class=MaterialExpressionLinearInterpolate
   inputs [
    A {from=MaterialExpressionConstant_30}
    B {from=MaterialExpressionConstant_31}
    Alpha {from=MaterialExpressionReroute_8, output=Alpha2}
   ]
  }
  MaterialExpressionLinearInterpolate_9 {
   class=MaterialExpressionLinearInterpolate
   inputs [
    A {from=MaterialExpressionMultiply_13}
    B {from=MaterialExpressionScalarParameter_5}
    Alpha {from=MaterialExpressionNamedRerouteUsage_11}
   ]
  }
  MaterialExpressionMaterialFunctionCall_0 {
   class=MaterialExpressionMaterialFunctionCall
   inputs [
    A {from=""}
    B {from=""}
    Transition Phase {from=MaterialExpressionDivide_2}
    Height Texture {from=MaterialExpressionLinearInterpolate_4}
    Contrast {from=""}
   ]
   properties {materialFunction=&HeightLerp}
  }
  MaterialExpressionMaterialFunctionCall_10 {
   class=MaterialExpressionMaterialFunctionCall
   inputs [Float3 {from=MaterialExpressionStaticSwitchParameter_10}]
   properties {materialFunction=&BreakOutFloat3Components}
  }
  MaterialExpressionMaterialFunctionCall_11 {
   class=MaterialExpressionMaterialFunctionCall
   inputs [
    Diffuse Texture Object {from=MaterialExpressionTextureObjectParameter_5}
    Spec Texture Object {from=MaterialExpressionTextureObjectParameter_6}
    TextureSize {from=MaterialExpressionScalarParameter_15}
    -------------------------- {from=""}
    Use seperate settings for Z projection {from=""}
    Use seperate texture for Z projection {from=""}
    Texture Size Z projection {from=""}
    Z Plane Diffuse Texture Object {from=""}
    Z Plane Spec Texture Object {from=""}
    "--------------------------      " {from=""}
    ProjectionTransitionContrast {from=""}
    Use Texture Alpha Heightmap to Refine the Transistion {from=""}
    HeightLerpTransitionContrast {from=""}
    PickTheDiffuseHeightChannel {from=""}
    TheDiffuseHeightChannel {from=""}
    "--------------------------       " {from=""}
    NormalTextureObject {from=MaterialExpressionTextureObjectParameter_7}
    World Space Normals {from=MaterialExpressionStaticBool_6}
    "--------------------------  " {from=""}
    Export Float 4 Diffuse {from=MaterialExpressionStaticBool_10}
    Export Float 4 Spec {from=""}
    "--------------------------     " {from=""}
    WorldPosition {from=""}
    World Space Normal {from=""}
   ]
   properties {materialFunction=/Engine/Functions/Engine_MaterialFunctions02/Texturing/WorldAlignedTextures_Complex}
  }
  MaterialExpressionMaterialFunctionCall_12 {
   class=MaterialExpressionMaterialFunctionCall
   inputs [Float3 {from=MaterialExpressionMaterialFunctionCall_11, output=Z Spec}]
   properties {materialFunction=&BreakOutFloat3Components}
  }
  MaterialExpressionMaterialFunctionCall_13 {
   class=MaterialExpressionMaterialFunctionCall
   inputs [Float3 {from=MaterialExpressionStaticSwitchParameter_11}]
   properties {materialFunction=&BreakOutFloat3Components}
  }
  MaterialExpressionMaterialFunctionCall_14 {
   class=MaterialExpressionMaterialFunctionCall
   inputs [
    A {from=""}
    B {from=""}
    Transition Phase {from=MaterialExpressionStaticSwitchParameter_21}
    Height Texture {from=MaterialExpressionLinearInterpolate_22}
    Contrast {from=MaterialExpressionScalarParameter_21}
   ]
   properties {materialFunction=&HeightLerp}
  }
  MaterialExpressionMaterialFunctionCall_15 {
   class=MaterialExpressionMaterialFunctionCall
   inputs [
    Diffuse Texture Object {from=MaterialExpressionTextureObject_0}
    Spec Texture Object {from=MaterialExpressionTextureObject_7}
    TextureSize {from=MaterialExpressionMultiply_20}
    -------------------------- {from=""}
    Use seperate settings for Z projection {from=""}
    Use seperate texture for Z projection {from=""}
    Texture Size Z projection {from=""}
    Z Plane Diffuse Texture Object {from=""}
    Z Plane Spec Texture Object {from=""}
    "--------------------------      " {from=""}
    ProjectionTransitionContrast {from=MaterialExpressionScalarParameter_4}
    Use Texture Alpha Heightmap to Refine the Transistion {from=""}
    HeightLerpTransitionContrast {from=""}
    PickTheDiffuseHeightChannel {from=""}
    TheDiffuseHeightChannel {from=""}
    "--------------------------       " {from=""}
    NormalTextureObject {from=MaterialExpressionTextureObject_6}
    World Space Normals {from=MaterialExpressionStaticBool_4}
    "--------------------------  " {from=""}
    Export Float 4 Diffuse {from=MaterialExpressionStaticBool_7}
    Export Float 4 Spec {from=""}
    "--------------------------     " {from=""}
    WorldPosition {from=""}
    World Space Normal {from=""}
   ]
   properties {materialFunction=&WorldAlignedTextures_Complex}
  }
  MaterialExpressionMaterialFunctionCall_16 {
   class=MaterialExpressionMaterialFunctionCall
   inputs [Float3 {from=MaterialExpressionNamedRerouteDeclaration_6}]
   properties {materialFunction=&BreakOutFloat3Components}
  }
  MaterialExpressionMaterialFunctionCall_17 {
   class=MaterialExpressionMaterialFunctionCall
   inputs [
    NoRVT {from=MaterialExpressionStaticSwitchParameter_3}
    RVT {from=MaterialExpressionFunctionInput_0}
   ]
   properties {materialFunction=/Game/Packages/Fortress_SharedMaterials/Base_Material_Functions/FN_RVT_Switch_Scalar}
  }
  MaterialExpressionMaterialFunctionCall_19 {
   class=MaterialExpressionMaterialFunctionCall
   inputs [Float3 {from=MaterialExpressionMaterialFunctionCall_11, output=Normal Z Texture}]
   properties {materialFunction=&BreakOutFloat3Components}
  }
  MaterialExpressionMaterialFunctionCall_2 {
   class=MaterialExpressionMaterialFunctionCall
   inputs [
    A {from=""}
    B {from=""}
    Transition Phase {from=MaterialExpressionSaturate_8}
    Height Texture {from=MaterialExpressionNamedRerouteUsage_6}
    Contrast {from=""}
   ]
   properties {materialFunction=&HeightLerp}
  }
  MaterialExpressionMaterialFunctionCall_26 {
   class=MaterialExpressionMaterialFunctionCall
   inputs [
    In {from=MaterialExpressionOneMinus_7}
    Contrast {from=MaterialExpressionConstant_28}
   ]
   properties {materialFunction=&CheapContrast}
  }
  MaterialExpressionMaterialFunctionCall_3 {
   class=MaterialExpressionMaterialFunctionCall
   inputs []
   properties {materialFunction=&ObjectLocalBounds}
  }
  MaterialExpressionMaterialFunctionCall_4 {
   class=MaterialExpressionMaterialFunctionCall
   inputs [
    In {from=MaterialExpressionAbs_1}
    Contrast {from=MaterialExpressionConstant_21}
   ]
   properties {materialFunction=&CheapContrast}
  }
  MaterialExpressionMaterialFunctionCall_5 {
   class=MaterialExpressionMaterialFunctionCall
   inputs [
    Diffuse Texture Object {from=MaterialExpressionTextureObjectParameter_8}
    Spec Texture Object {from=MaterialExpressionTextureObjectParameter_9}
    TextureSize {from=MaterialExpressionScalarParameter_10}
    -------------------------- {from=""}
    Use seperate settings for Z projection {from=""}
    Use seperate texture for Z projection {from=""}
    Texture Size Z projection {from=""}
    Z Plane Diffuse Texture Object {from=""}
    Z Plane Spec Texture Object {from=""}
    "--------------------------      " {from=""}
    ProjectionTransitionContrast {from=""}
    Use Texture Alpha Heightmap to Refine the Transistion {from=""}
    HeightLerpTransitionContrast {from=""}
    PickTheDiffuseHeightChannel {from=""}
    TheDiffuseHeightChannel {from=""}
    "--------------------------       " {from=""}
    NormalTextureObject {from=MaterialExpressionTextureObjectParameter_12}
    World Space Normals {from=MaterialExpressionStaticBool_2}
    "--------------------------  " {from=""}
    Export Float 4 Diffuse {from=MaterialExpressionStaticBool_3}
    Export Float 4 Spec {from=""}
    "--------------------------     " {from=""}
    WorldPosition {from=""}
    World Space Normal {from=""}
   ]
   properties {materialFunction=&WorldAlignedTextures_Complex}
  }
  MaterialExpressionMaterialFunctionCall_6 {
   class=MaterialExpressionMaterialFunctionCall
   inputs [
    Diffuse Texture Object {from=MaterialExpressionTextureObjectParameter_2}
    Spec Texture Object {from=MaterialExpressionTextureObjectParameter_3}
    TextureSize {from=MaterialExpressionScalarParameter_13}
    -------------------------- {from=""}
    Use seperate settings for Z projection {from=""}
    Use seperate texture for Z projection {from=""}
    Texture Size Z projection {from=""}
    Z Plane Diffuse Texture Object {from=""}
    Z Plane Spec Texture Object {from=""}
    "--------------------------      " {from=""}
    ProjectionTransitionContrast {from=MaterialExpressionScalarParameter_27}
    Use Texture Alpha Heightmap to Refine the Transistion {from=""}
    HeightLerpTransitionContrast {from=""}
    PickTheDiffuseHeightChannel {from=""}
    TheDiffuseHeightChannel {from=""}
    "--------------------------       " {from=""}
    NormalTextureObject {from=MaterialExpressionTextureObjectParameter_4}
    World Space Normals {from=MaterialExpressionStaticBool_1}
    "--------------------------  " {from=""}
    Export Float 4 Diffuse {from=MaterialExpressionStaticBool_8}
    Export Float 4 Spec {from=""}
    "--------------------------     " {from=""}
    WorldPosition {from=""}
    World Space Normal {from=""}
   ]
   properties {materialFunction=&WorldAlignedTextures_Complex}
  }
  MaterialExpressionMaterialFunctionCall_9 {
   class=MaterialExpressionMaterialFunctionCall
   inputs [Float3 {from=MaterialExpressionNamedRerouteDeclaration_5}]
   properties {materialFunction=&BreakOutFloat3Components}
  }
  MaterialExpressionMax_0 {
   class=MaterialExpressionMax
   inputs [
    A {from=MaterialExpressionNamedRerouteDeclaration_0}
    B {from=MaterialExpressionConstant_4}
   ]
  }
  MaterialExpressionMin_1 {
   class=MaterialExpressionMin
   inputs [
    A {from=MaterialExpressionNamedRerouteUsage_61, output=""}
    B {from=MaterialExpressionNamedRerouteUsage_76, output=""}
   ]
   properties {}
  }
  MaterialExpressionMin_7 {
   class=MaterialExpressionMin
   inputs [
    A {from=MaterialExpressionNamedRerouteUsage_85, output=""}
    B {from=MaterialExpressionComponentMask_18, output=""}
   ]
   properties {}
  }
  MaterialExpressionMultiply_0 {
   class=MaterialExpressionMultiply
   inputs [
    A {from=MaterialExpressionConstant_3}
    B {from=MaterialExpressionLinearInterpolate_13}
   ]
  }
  MaterialExpressionMultiply_1 {
   class=MaterialExpressionMultiply
   inputs [
    A {from=""}
    B {from=MaterialExpressionScalarParameter_23}
   ]
   properties {constA=0.25}
  }
  MaterialExpressionMultiply_11 {
   class=MaterialExpressionMultiply
   inputs [
    A {from=MaterialExpressionVectorParameter_13, output=Spots Value}
    B {from=MaterialExpressionNamedRerouteUsage_8}
   ]
  }
  MaterialExpressionMultiply_12 {
   class=MaterialExpressionMultiply
   inputs [
    A {from=MaterialExpressionGetMaterialAttributes_3, output=BaseColor}
    B {from=MaterialExpressionVectorParameter_1}
   ]
  }
  MaterialExpressionMultiply_13 {
   class=MaterialExpressionMultiply
   inputs [
    A {from=MaterialExpressionScalarParameter_5}
    B {from=""}
   ]
   properties {constB=3}
  }
  MaterialExpressionMultiply_15 {
   class=MaterialExpressionMultiply
   inputs [
    A {from=MaterialExpressionDesaturation_0}
    B {from=MaterialExpressionVectorParameter_8}
   ]
  }
  MaterialExpressionMultiply_16 {
   class=MaterialExpressionMultiply
   inputs [
    A {from=MaterialExpressionReroute_5}
    B {from=MaterialExpressionConstant3Vector_6}
   ]
  }
  MaterialExpressionMultiply_17 {
   class=MaterialExpressionMultiply
   inputs [
    A {from=MaterialExpressionMultiply_16}
    B {from=MaterialExpressionLinearInterpolate_1}
   ]
  }
  MaterialExpressionMultiply_18 {
   class=MaterialExpressionMultiply
   inputs [
    A {from=MaterialExpressionOneMinus_2}
    B {from=MaterialExpressionScalarParameter_32}
   ]
  }
  MaterialExpressionMultiply_19 {
   class=MaterialExpressionMultiply
   inputs [
    A {from=MaterialExpressionStaticSwitchParameter_16}
    B {from=MaterialExpressionLinearInterpolate_12}
   ]
  }
  MaterialExpressionMultiply_2 {
   class=MaterialExpressionMultiply
   inputs [
    A {from=MaterialExpressionMaterialFunctionCall_0, output=Lerp Alpha No Contrast}
    B {from=MaterialExpressionMax_0}
   ]
  }
  MaterialExpressionMultiply_20 {
   class=MaterialExpressionMultiply
   inputs [
    A {from=MaterialExpressionConstant_45}
    B {from=MaterialExpressionScalarParameter_9}
   ]
  }
  MaterialExpressionMultiply_21 {
   class=MaterialExpressionMultiply
   inputs [
    A {from=""}
    B {from=MaterialExpressionScalarParameter_16}
   ]
   properties {constA=0.25}
  }
  MaterialExpressionMultiply_22 {
   class=MaterialExpressionMultiply
   inputs [
    A {from=MaterialExpressionLinearInterpolate_21}
    B {from=MaterialExpressionAppendVector_6}
   ]
  }
  MaterialExpressionMultiply_23 {
   class=MaterialExpressionMultiply
   inputs [
    A {from=MaterialExpressionGetMaterialAttributes_7, output=Specular}
    B {from=MaterialExpressionScalarParameter_31}
   ]
  }
  MaterialExpressionMultiply_24 {
   class=MaterialExpressionMultiply
   inputs [
    A {from=MaterialExpressionLinearInterpolate_3}
    B {from=MaterialExpressionAppendVector_2}
   ]
  }
  MaterialExpressionMultiply_25 {
   class=MaterialExpressionMultiply
   inputs [
    A {from=MaterialExpressionSaturate_10}
    B {from=MaterialExpressionNamedRerouteUsage_36}
   ]
  }
  MaterialExpressionMultiply_26 {
   class=MaterialExpressionMultiply
   inputs [
    A {from=MaterialExpressionComponentMask_15}
    B {from=MaterialExpressionAbs_4}
   ]
  }
  MaterialExpressionMultiply_27 {
   class=MaterialExpressionMultiply
   inputs [
    A {from=MaterialExpressionVectorParameter_6, output=RGB}
    B {from=MaterialExpressionDesaturation_2}
   ]
  }
  MaterialExpressionMultiply_28 {
   class=MaterialExpressionMultiply
   inputs [
    A {from=MaterialExpressionGetMaterialAttributes_1, output=Specular}
    B {from=MaterialExpressionScalarParameter_34}
   ]
  }
  MaterialExpressionMultiply_29 {
   class=MaterialExpressionMultiply
   inputs [
    A {from=MaterialExpressionGetMaterialAttributes_7, output=Roughness}
    B {from=MaterialExpressionLinearInterpolate_23}
   ]
  }
  MaterialExpressionMultiply_3 {
   class=MaterialExpressionMultiply
   inputs [
    A {from=MaterialExpressionLinearInterpolate_9}
    B {from=MaterialExpressionConstant_6}
   ]
   properties {constB=5}
  }
  MaterialExpressionMultiply_31 {
   class=MaterialExpressionMultiply
   inputs [
    A {from=MaterialExpressionMultiply_56}
    B {from=MaterialExpressionSaturate_11}
   ]
  }
  MaterialExpressionMultiply_32 {
   class=MaterialExpressionMultiply
   inputs [
    A {from=MaterialExpressionStaticSwitchParameter_13}
    B {from=MaterialExpressionSaturate_12}
   ]
  }
  MaterialExpressionMultiply_33 {
   class=MaterialExpressionMultiply
   inputs [
    A {from=MaterialExpressionScalarParameter_38}
    B {from=MaterialExpressionTextureSampleParameter2D_1, output=R}
   ]
   properties {constB=3.25}
  }
  MaterialExpressionMultiply_34 {
   class=MaterialExpressionMultiply
   inputs [
    A {from=MaterialExpressionScalarParameter_36}
    B {from=MaterialExpressionTextureSampleParameter2D_0, output=RGB}
   ]
   properties {constA=2.5}
  }
  MaterialExpressionMultiply_35 {
   class=MaterialExpressionMultiply
   inputs [
    A {from=MaterialExpressionGetMaterialAttributes_7, output=BaseColor}
    B {from=MaterialExpressionLinearInterpolate_25}
   ]
  }
  MaterialExpressionMultiply_38 {
   class=MaterialExpressionMultiply
   inputs [
    A {from=MaterialExpressionScalarParameter_25}
    B {from=MaterialExpressionAppendVector_8}
   ]
  }
  MaterialExpressionMultiply_4 {
   class=MaterialExpressionMultiply
   inputs [
    A {from=MaterialExpressionGetMaterialAttributes_0, output=Specular}
    B {from=MaterialExpressionConstant_16}
   ]
  }
  MaterialExpressionMultiply_5 {
   class=MaterialExpressionMultiply
   inputs [
    A {from=MaterialExpressionVectorParameter_0}
    B {from=MaterialExpressionGetMaterialAttributes_2, output=BaseColor}
   ]
  }
  MaterialExpressionMultiply_52 {
   class=MaterialExpressionMultiply
   inputs [
    A {from=MaterialExpressionGetMaterialAttributes_7, output=BaseColor}
    B {from=MaterialExpressionLinearInterpolate_34}
   ]
  }
  MaterialExpressionMultiply_56 {
   class=MaterialExpressionMultiply
   inputs [
    A {from=MaterialExpressionMultiply_11}
    B {from=MaterialExpressionMaterialFunctionCall_26, output=Result}
   ]
  }
  MaterialExpressionMultiply_62 {
   class=MaterialExpressionMultiply
   inputs [
    A {from=MaterialExpressionAbs_2}
    B {from=""}
   ]
   properties {constB=-1}
  }
  MaterialExpressionMultiply_64 {
   class=MaterialExpressionMultiply
   inputs [
    A {from=""}
    B {from=MaterialExpressionMultiply_20}
   ]
   properties {constA=2, constB=2}
  }
  MaterialExpressionMultiply_65 {
   class=MaterialExpressionMultiply
   inputs [
    A {from=MaterialExpressionVectorParameter_16, output=Dust Ammount}
    B {from=MaterialExpressionOneMinus_8}
   ]
  }
  MaterialExpressionMultiply_7 {
   class=MaterialExpressionMultiply
   inputs [
    A {from=MaterialExpressionReroute_16, output=A}
    B {from=MaterialExpressionComponentMask_0}
   ]
  }
  MaterialExpressionMultiply_8 {
   class=MaterialExpressionMultiply
   inputs [
    A {from=MaterialExpressionAppendVector_0}
    B {from=MaterialExpressionLinearInterpolate_11}
   ]
  }
  MaterialExpressionMultiply_9 {
   class=MaterialExpressionMultiply
   inputs [
    A {from=MaterialExpressionVectorParameter_8, output=Roughness Tweal}
    B {from=MaterialExpressionMaterialFunctionCall_12, output=B}
   ]
  }
  MaterialExpressionNamedRerouteDeclaration_0 {
   class=MaterialExpressionNamedRerouteDeclaration
   inputs [None {from=MaterialExpressionSaturate_2}]
   properties {name=Soft Waterline, nodeColor=<rgba 0 0.5312 1 1>}
  }
  MaterialExpressionNamedRerouteDeclaration_1 {
   class=MaterialExpressionNamedRerouteDeclaration
   inputs [None {from=MaterialExpressionSetMaterialAttributes_2}]
   properties {name=Baked Set, nodeColor=<rgba 1 0 0.7969 1>}
  }
  MaterialExpressionNamedRerouteDeclaration_10 {
   class=MaterialExpressionNamedRerouteDeclaration
   inputs [None {from=MaterialExpressionComponentMask_6}]
   properties {name=Side L2 Alpha, nodeColor=<rgba 1 0.2812 0 1>}
  }
  MaterialExpressionNamedRerouteDeclaration_11 {
   class=MaterialExpressionNamedRerouteDeclaration
   inputs [None {from=MaterialExpressionComponentMask_8}]
   properties {name=Side A, nodeColor=<rgba 0.6641 1 0 1>}
  }
  MaterialExpressionNamedRerouteDeclaration_12 {
   class=MaterialExpressionNamedRerouteDeclaration
   inputs [None {from=MaterialExpressionFunctionInput_2}]
   properties {name=Up Layer outside, nodeColor=<rgba 0.4062 0 1 1>}
  }
  MaterialExpressionNamedRerouteDeclaration_13 {
   class=MaterialExpressionNamedRerouteDeclaration
   inputs [None {from=MaterialExpressionSaturate_7}]
   properties {name=RVT Mask, nodeColor=<rgba 1 0.8672 0 1>}
  }
  MaterialExpressionNamedRerouteDeclaration_15 {
   class=MaterialExpressionNamedRerouteDeclaration
   inputs [None {from=MaterialExpressionQualitySwitch_0}]
   properties {name=RocksNew, nodeColor=<rgba 0.2422 0 1 1>}
  }
  MaterialExpressionNamedRerouteDeclaration_16 {
   class=MaterialExpressionNamedRerouteDeclaration
   inputs [None {from=MaterialExpressionReroute_6}]
   properties {name=Details XYZ RGB Mask, nodeColor=<rgba 1 0 0.6094 1>}
  }
  MaterialExpressionNamedRerouteDeclaration_17 {
   class=MaterialExpressionNamedRerouteDeclaration
   inputs [None {from=MaterialExpressionSetMaterialAttributes_7}]
   properties {name=Sides Clean, nodeColor=<rgba 0 0.8828 1 1>}
  }
  MaterialExpressionNamedRerouteDeclaration_18 {
   class=MaterialExpressionNamedRerouteDeclaration
   inputs [None {from=MaterialExpressionQualitySwitch_3}]
   properties {name=Details Spots, nodeColor=<rgba 1 0 0.8906 1>}
  }
  MaterialExpressionNamedRerouteDeclaration_19 {
   class=MaterialExpressionNamedRerouteDeclaration
   inputs [None {from=MaterialExpressionTextureSampleParameter2D_1, output=G}]
   properties {name=Baked Metal Mask, nodeColor=<rgba 0.9922 0 1 1>}
  }
  MaterialExpressionNamedRerouteDeclaration_2 {
   class=MaterialExpressionNamedRerouteDeclaration
   inputs [None {from=MaterialExpressionComponentMask_12}]
   properties {name=Diffuse fo snow, nodeColor=<rgba 0.3828 0 1 1>}
  }
  MaterialExpressionNamedRerouteDeclaration_20 {
   class=MaterialExpressionNamedRerouteDeclaration
   inputs [None {from=MaterialExpressionStaticSwitchParameter_8}]
   properties {name=Side Alphas, nodeColor=<rgba 0.5937 1 0 1>}
  }
  MaterialExpressionNamedRerouteDeclaration_21 {
   class=MaterialExpressionNamedRerouteDeclaration
   inputs [None {from=MaterialExpressionComponentMask_13}]
   properties {name=Top A, nodeColor=<rgba 0 1 0.2031 1>}
  }
  MaterialExpressionNamedRerouteDeclaration_22 {
   class=MaterialExpressionNamedRerouteDeclaration
   inputs [None {from=MaterialExpressionSetMaterialAttributes_11}]
   properties {name=RocksPlusDetails, nodeColor=<rgba 1 0.7031 0 1>}
  }
  MaterialExpressionNamedRerouteDeclaration_23 {
   class=MaterialExpressionNamedRerouteDeclaration
   inputs [None {from=MaterialExpressionSaturate_4}]
   properties {name=Rocks Side/UpMask, nodeColor=<rgba 0 1 0.2031 1>}
  }
  MaterialExpressionNamedRerouteDeclaration_3 {
   class=MaterialExpressionNamedRerouteDeclaration
   inputs [None {from=MaterialExpressionSaturate_0}]
   properties {name=AO, nodeColor=<rgba 1 0 0.2812 1>}
  }
  MaterialExpressionNamedRerouteDeclaration_30 {
   class=MaterialExpressionNamedRerouteDeclaration
   inputs [None {from=MaterialExpressionStaticSwitch_2}]
   properties {name=Rocks Side Mat, nodeColor=<rgba 0.3594 0 1 1>}
  }
  MaterialExpressionNamedRerouteDeclaration_31 {
   class=MaterialExpressionNamedRerouteDeclaration
   inputs [None {from=MaterialExpressionStaticSwitchParameter_7}]
   properties {name=DetailsN To Rocks, nodeColor=<rgba 0 0.4609 1 1>}
  }
  MaterialExpressionNamedRerouteDeclaration_32 {
   class=MaterialExpressionNamedRerouteDeclaration
   inputs [None {from=MaterialExpressionStaticSwitchParameter_15}]
   properties {name=Rocks Top Mat, nodeColor=<rgba 0.5234 1 0 1>}
  }
  MaterialExpressionNamedRerouteDeclaration_35 {
   class=MaterialExpressionNamedRerouteDeclaration
   inputs [None {from=MaterialExpressionStaticSwitch_7}]
   properties {name=Details Moss, nodeColor=<rgba 1 0 0.8906 1>}
  }
  MaterialExpressionNamedRerouteDeclaration_37 {
   class=MaterialExpressionNamedRerouteDeclaration
   inputs [None {from=MaterialExpressionStaticSwitchParameter_1}]
   properties {name=Top+Moss+Spots, nodeColor=<rgba 1 0 0.5391 1>}
  }
  MaterialExpressionNamedRerouteDeclaration_39 {
   class=MaterialExpressionNamedRerouteDeclaration
   inputs [None {from=MaterialExpressionSetMaterialAttributes_16}]
   properties {name=Side*colors, nodeColor=<rgba 0 0.6484 1 1>}
  }
  MaterialExpressionNamedRerouteDeclaration_4 {
   class=MaterialExpressionNamedRerouteDeclaration
   inputs [None {from=MaterialExpressionMaterialFunctionCall_6, output=XY Diffuse Texture}]
   properties {name=Sides XY D, nodeColor=<rgba 1 0 0.1172 1>}
  }
  MaterialExpressionNamedRerouteDeclaration_40 {
   class=MaterialExpressionNamedRerouteDeclaration
   inputs [None {from=MaterialExpressionBlendMaterialAttributes_10}]
   properties {name=Rocks with Dust, nodeColor=<rgba 0.1484 1 0 1>}
  }
  MaterialExpressionNamedRerouteDeclaration_41 {
   class=MaterialExpressionNamedRerouteDeclaration
   inputs [None {from=MaterialExpressionSaturate_15}]
   properties {name=Details Sharp mask, nodeColor=<rgba 1 0.7031 0 1>}
  }
  MaterialExpressionNamedRerouteDeclaration_42 {
   class=MaterialExpressionNamedRerouteDeclaration
   inputs [None {from=MaterialExpressionDesaturation_1}]
   properties {name=Rocks Side Dif, nodeColor=<rgba 0 0.9297 1 1>}
  }
  MaterialExpressionNamedRerouteDeclaration_46 {
   class=MaterialExpressionNamedRerouteDeclaration
   inputs [None {from=MaterialExpressionQualitySwitch_2}]
   properties {name=Moss Details Scale 2, nodeColor=<rgba 0.2656 0 1 1>}
  }
  MaterialExpressionNamedRerouteDeclaration_5 {
   class=MaterialExpressionNamedRerouteDeclaration
   inputs [None {from=MaterialExpressionMaterialFunctionCall_6, output=XY Spec}]
   properties {name=Sides XY S, nodeColor=<rgba 0.8047 1 0 1>}
  }
  MaterialExpressionNamedRerouteDeclaration_6 {
   class=MaterialExpressionNamedRerouteDeclaration
   inputs [None {from=MaterialExpressionMaterialFunctionCall_6, output=Normal XY Texture}]
   properties {name=Sides XY N, nodeColor=<rgba 0.7109 0 1 1>}
  }
  MaterialExpressionNamedRerouteDeclaration_9 {
   class=MaterialExpressionNamedRerouteDeclaration
   inputs [None {from=MaterialExpressionMaterialFunctionCall_14, output=Alpha}]
   properties {name=Mask for dual layer, nodeColor=<rgba 1 0.7031 0 1>}
  }
  MaterialExpressionNamedRerouteUsage_0 {class=MaterialExpressionNamedRerouteUsage}
  MaterialExpressionNamedRerouteUsage_11 {class=MaterialExpressionNamedRerouteUsage}
  MaterialExpressionNamedRerouteUsage_12 {class=MaterialExpressionNamedRerouteUsage}
  MaterialExpressionNamedRerouteUsage_13 {class=MaterialExpressionNamedRerouteUsage}
  MaterialExpressionNamedRerouteUsage_14 {class=MaterialExpressionNamedRerouteUsage}
  MaterialExpressionNamedRerouteUsage_15 {class=MaterialExpressionNamedRerouteUsage}
  MaterialExpressionNamedRerouteUsage_16 {class=MaterialExpressionNamedRerouteUsage}
  MaterialExpressionNamedRerouteUsage_17 {class=MaterialExpressionNamedRerouteUsage}
  MaterialExpressionNamedRerouteUsage_18 {class=MaterialExpressionNamedRerouteUsage}
  MaterialExpressionNamedRerouteUsage_19 {class=MaterialExpressionNamedRerouteUsage}
  MaterialExpressionNamedRerouteUsage_2 {class=MaterialExpressionNamedRerouteUsage}
  MaterialExpressionNamedRerouteUsage_22 {class=MaterialExpressionNamedRerouteUsage}
  MaterialExpressionNamedRerouteUsage_23 {class=MaterialExpressionNamedRerouteUsage}
  MaterialExpressionNamedRerouteUsage_24 {class=MaterialExpressionNamedRerouteUsage}
  MaterialExpressionNamedRerouteUsage_26 {class=MaterialExpressionNamedRerouteUsage}
  MaterialExpressionNamedRerouteUsage_27 {class=MaterialExpressionNamedRerouteUsage}
  MaterialExpressionNamedRerouteUsage_28 {class=MaterialExpressionNamedRerouteUsage}
  MaterialExpressionNamedRerouteUsage_29 {class=MaterialExpressionNamedRerouteUsage}
  MaterialExpressionNamedRerouteUsage_3 {class=MaterialExpressionNamedRerouteUsage}
  MaterialExpressionNamedRerouteUsage_30 {class=MaterialExpressionNamedRerouteUsage}
  MaterialExpressionNamedRerouteUsage_31 {class=MaterialExpressionNamedRerouteUsage}
  MaterialExpressionNamedRerouteUsage_32 {class=MaterialExpressionNamedRerouteUsage}
  MaterialExpressionNamedRerouteUsage_33 {class=MaterialExpressionNamedRerouteUsage}
  MaterialExpressionNamedRerouteUsage_34 {class=MaterialExpressionNamedRerouteUsage}
  MaterialExpressionNamedRerouteUsage_36 {class=MaterialExpressionNamedRerouteUsage}
  MaterialExpressionNamedRerouteUsage_37 {class=MaterialExpressionNamedRerouteUsage}
  MaterialExpressionNamedRerouteUsage_38 {class=MaterialExpressionNamedRerouteUsage}
  MaterialExpressionNamedRerouteUsage_39 {class=MaterialExpressionNamedRerouteUsage}
  MaterialExpressionNamedRerouteUsage_4 {class=MaterialExpressionNamedRerouteUsage}
  MaterialExpressionNamedRerouteUsage_42 {class=MaterialExpressionNamedRerouteUsage}
  MaterialExpressionNamedRerouteUsage_45 {class=MaterialExpressionNamedRerouteUsage}
  MaterialExpressionNamedRerouteUsage_49 {class=MaterialExpressionNamedRerouteUsage}
  MaterialExpressionNamedRerouteUsage_5 {class=MaterialExpressionNamedRerouteUsage}
  MaterialExpressionNamedRerouteUsage_52 {class=MaterialExpressionNamedRerouteUsage}
  MaterialExpressionNamedRerouteUsage_57 {class=MaterialExpressionNamedRerouteUsage}
  MaterialExpressionNamedRerouteUsage_6 {class=MaterialExpressionNamedRerouteUsage}
  MaterialExpressionNamedRerouteUsage_61 {class=MaterialExpressionNamedRerouteUsage}
  MaterialExpressionNamedRerouteUsage_74 {class=MaterialExpressionNamedRerouteUsage}
  MaterialExpressionNamedRerouteUsage_76 {class=MaterialExpressionNamedRerouteUsage}
  MaterialExpressionNamedRerouteUsage_8 {class=MaterialExpressionNamedRerouteUsage}
  MaterialExpressionNamedRerouteUsage_85 {class=MaterialExpressionNamedRerouteUsage}
  MaterialExpressionOneMinus_1 {class=MaterialExpressionOneMinus, inputs [None {from=MaterialExpressionSphereMask_0}]}
  MaterialExpressionOneMinus_2 {class=MaterialExpressionOneMinus, inputs [None {from=MaterialExpressionSaturate_16}]}
  MaterialExpressionOneMinus_5 {class=MaterialExpressionOneMinus, inputs [None {from=MaterialExpressionDotProduct_0}]}
  MaterialExpressionOneMinus_7 {class=MaterialExpressionOneMinus, inputs [None {from=MaterialExpressionNamedRerouteUsage_49}]}
  MaterialExpressionOneMinus_8 {class=MaterialExpressionOneMinus, inputs [None {from=MaterialExpressionNamedRerouteUsage_61}]}
  MaterialExpressionPinBase_0 {
   class=MaterialExpressionPinBase
   inputs []
   properties {reroutePins [
     Alpha {expression=&self:MaterialExpressionReroute_0}
     Alpha2 {expression=&self:MaterialExpressionReroute_8}
     Alpha3 {expression=&self:MaterialExpressionReroute_9}
    ]}
  }
  MaterialExpressionPinBase_1 {
   class=MaterialExpressionPinBase
   inputs [Output {from=MaterialExpressionLinearInterpolate_26, output=""}]
   properties {reroutePins [Output {expression=&self:MaterialExpressionReroute_7}]}
  }
  MaterialExpressionPinBase_2 {
   class=MaterialExpressionPinBase
   inputs []
   properties {reroutePins [
     Alpha {expression=&self:MaterialExpressionReroute_10}
     Alpha2 {expression=&self:MaterialExpressionReroute_12}
     Alpha3 {expression=&self:MaterialExpressionReroute_13}
    ]}
  }
  MaterialExpressionPinBase_3 {
   class=MaterialExpressionPinBase
   inputs [Output {from=MaterialExpressionLinearInterpolate_29, output=""}]
   properties {reroutePins [Output {expression=&self:MaterialExpressionReroute_11}]}
  }
  MaterialExpressionPinBase_4 {class=MaterialExpressionPinBase, inputs [], properties {}}
  MaterialExpressionPinBase_5 {
   class=MaterialExpressionPinBase
   inputs [Output {from=MaterialExpressionLinearInterpolate_19, output=""}]
   properties {reroutePins [Output {expression=&self:MaterialExpressionReroute_14}]}
  }
  MaterialExpressionPinBase_6 {
   class=MaterialExpressionPinBase
   inputs []
   properties {reroutePins [
     A {expression=&self:MaterialExpressionReroute_16}
     Input {expression=&self:MaterialExpressionReroute_17}
     Alpha {expression=&self:MaterialExpressionReroute_18}
     Alpha2 {expression=&self:MaterialExpressionReroute_19}
     Input2 {expression=&self:MaterialExpressionReroute_20}
    ]}
  }
  MaterialExpressionPinBase_7 {
   class=MaterialExpressionPinBase
   inputs [Output {from=MaterialExpressionMultiply_8, output=""}]
   properties {reroutePins [Output {expression=&self:MaterialExpressionReroute_15}]}
  }
  MaterialExpressionPower_0 {
   class=MaterialExpressionPower
   inputs [
    Base {from=MaterialExpressionDesaturation_1}
    Exp {from=MaterialExpressionScalarParameter_33}
   ]
  }
  MaterialExpressionQualitySwitch_0 {
   class=MaterialExpressionQualitySwitch
   inputs [
    Default {from=MaterialExpressionReroute_24, output=""}
    Low {from="", output=""}
    High {from="", output=""}
    Medium {from="", output=""}
    Epic {from=MaterialExpressionNamedRerouteDeclaration_22, output=""}
   ]
   properties {}
  }
  MaterialExpressionQualitySwitch_1 {
   class=MaterialExpressionQualitySwitch
   inputs [
    Default {from=MaterialExpressionConstant_22, output=""}
    Low {from="", output=""}
    High {from="", output=""}
    Medium {from="", output=""}
    Epic {from=MaterialExpressionShadingPathSwitch_0, output=""}
   ]
   properties {}
  }
  MaterialExpressionQualitySwitch_2 {
   class=MaterialExpressionQualitySwitch
   inputs [
    Default {from=MaterialExpressionConstant_18, output=""}
    Low {from="", output=""}
    High {from="", output=""}
    Medium {from="", output=""}
    Epic {from=MaterialExpressionTextureSample_2, output=B}
   ]
   properties {}
  }
  MaterialExpressionQualitySwitch_3 {
   class=MaterialExpressionQualitySwitch
   inputs [
    Default {from=MaterialExpressionConstant_18, output=""}
    Low {from="", output=""}
    High {from="", output=""}
    Medium {from="", output=""}
    Epic {from=MaterialExpressionTextureSample_2, output=R}
   ]
   properties {}
  }
  MaterialExpressionQualitySwitch_4 {
   class=MaterialExpressionQualitySwitch
   inputs [
    Default {from=MaterialExpressionNamedRerouteUsage_52, output=""}
    Low {from="", output=""}
    High {from="", output=""}
    Medium {from="", output=""}
    Epic {from=MaterialExpressionBlendMaterialAttributes_7, output=""}
   ]
   properties {}
  }
  MaterialExpressionReroute_0 {class=MaterialExpressionReroute, inputs [None {from=MaterialExpressionNamedRerouteUsage_35}]}
  MaterialExpressionReroute_1 {class=MaterialExpressionReroute, inputs [None {from=MaterialExpressionGetMaterialAttributes_0, output=MaterialAttributes}]}
  MaterialExpressionReroute_10 {class=MaterialExpressionReroute, inputs [None {from=MaterialExpressionNamedRerouteUsage_35}]}
  MaterialExpressionReroute_11 {class=MaterialExpressionReroute, inputs [None {from=MaterialExpressionLinearInterpolate_29}]}
  MaterialExpressionReroute_12 {class=MaterialExpressionReroute, inputs [None {from=MaterialExpressionMultiply_14}]}
  MaterialExpressionReroute_13 {class=MaterialExpressionReroute, inputs [None {from=MaterialExpressionMultiply_14}]}
  MaterialExpressionReroute_14 {class=MaterialExpressionReroute, inputs [None {from=MaterialExpressionLinearInterpolate_19}]}
  MaterialExpressionReroute_15 {class=MaterialExpressionReroute, inputs [None {from=MaterialExpressionMultiply_8}]}
  MaterialExpressionReroute_16 {class=MaterialExpressionReroute, inputs [None {from=MaterialExpressionScalarParameter_2}]}
  MaterialExpressionReroute_17 {class=MaterialExpressionReroute, inputs [None {from=MaterialExpressionNamedRerouteUsage_10}]}
  MaterialExpressionReroute_18 {class=MaterialExpressionReroute, inputs [None {from=MaterialExpressionNamedRerouteUsage_9}]}
  MaterialExpressionReroute_19 {class=MaterialExpressionReroute, inputs [None {from=MaterialExpressionNamedRerouteUsage_21}]}
  MaterialExpressionReroute_2 {class=MaterialExpressionReroute, inputs [None {from=MaterialExpressionReroute_1}]}
  MaterialExpressionReroute_20 {class=MaterialExpressionReroute, inputs [None {from=MaterialExpressionNamedRerouteUsage_10}]}
  MaterialExpressionReroute_21 {class=MaterialExpressionReroute, inputs [None {from=MaterialExpressionStaticSwitchParameter_5}]}
  MaterialExpressionReroute_22 {class=MaterialExpressionReroute, inputs [None {from=MaterialExpressionReroute_21}]}
  MaterialExpressionReroute_23 {class=MaterialExpressionReroute, inputs [None {from=MaterialExpressionMaterialFunctionCall_15, output=XYZ Diffuse Texture}]}
  MaterialExpressionReroute_24 {class=MaterialExpressionReroute, inputs [None {from=MaterialExpressionReroute_22}]}
  MaterialExpressionReroute_25 {class=MaterialExpressionReroute, inputs [None {from=MaterialExpressionNamedRerouteDeclaration_30}]}
  MaterialExpressionReroute_26 {class=MaterialExpressionReroute, inputs [None {from=MaterialExpressionStaticSwitchParameter_12}]}
  MaterialExpressionReroute_27 {class=MaterialExpressionReroute, inputs [None {from=MaterialExpressionStaticSwitchParameter_5}]}
  MaterialExpressionReroute_3 {class=MaterialExpressionReroute, inputs [None {from=MaterialExpressionStaticSwitchParameter_3}]}
  MaterialExpressionReroute_37 {class=MaterialExpressionReroute, inputs [None {from=MaterialExpressionNamedRerouteUsage_24}]}
  MaterialExpressionReroute_4 {class=MaterialExpressionReroute, inputs [None {from=MaterialExpressionMaterialFunctionCall_15, output=Normal XYZ Texture}]}
  MaterialExpressionReroute_5 {class=MaterialExpressionReroute, inputs [None {from=MaterialExpressionReroute_4}]}
  MaterialExpressionReroute_6 {class=MaterialExpressionReroute, inputs [None {from=MaterialExpressionMaterialFunctionCall_15, output=XYZ Spec}]}
  MaterialExpressionReroute_7 {class=MaterialExpressionReroute, inputs [None {from=MaterialExpressionLinearInterpolate_26}]}
  MaterialExpressionReroute_8 {class=MaterialExpressionReroute, inputs [None {from=MaterialExpressionMultiply_14}]}
  MaterialExpressionReroute_9 {class=MaterialExpressionReroute, inputs [None {from=MaterialExpressionMultiply_14}]}
  MaterialExpressionSaturate_0 {class=MaterialExpressionSaturate, inputs [None {from=MaterialExpressionFunctionOutput_1}]}
  MaterialExpressionSaturate_1 {class=MaterialExpressionSaturate, inputs [None {from=MaterialExpressionDivide_2}]}
  MaterialExpressionSaturate_10 {class=MaterialExpressionSaturate, inputs [None {from=MaterialExpressionDivide_6}]}
  MaterialExpressionSaturate_11 {class=MaterialExpressionSaturate, inputs [None {from=MaterialExpressionComponentMask_9}]}
  MaterialExpressionSaturate_12 {class=MaterialExpressionSaturate, inputs [None {from=MaterialExpressionLinearInterpolate_18}]}
  MaterialExpressionSaturate_13 {class=MaterialExpressionSaturate, inputs [None {from=MaterialExpressionStaticSwitch_5}]}
  MaterialExpressionSaturate_14 {class=MaterialExpressionSaturate, inputs [None {from=MaterialExpressionMultiply_31}]}
  MaterialExpressionSaturate_15 {class=MaterialExpressionSaturate, inputs [None {from=MaterialExpressionLinearInterpolate_36}]}
  MaterialExpressionSaturate_16 {class=MaterialExpressionSaturate, inputs [None {from=MaterialExpressionMultiply_26}]}
  MaterialExpressionSaturate_17 {class=MaterialExpressionSaturate, inputs [None {from=MaterialExpressionMultiply_34}]}
  MaterialExpressionSaturate_2 {class=MaterialExpressionSaturate, inputs [None {from=MaterialExpressionDivide_3}]}
  MaterialExpressionSaturate_21 {class=MaterialExpressionSaturate, inputs [None {from=MaterialExpressionComponentMask_30}]}
  MaterialExpressionSaturate_22 {class=MaterialExpressionSaturate, inputs [None {from=MaterialExpressionMultiply_65}]}
  MaterialExpressionSaturate_3 {class=MaterialExpressionSaturate, inputs [None {from=MaterialExpressionDivide_4}]}
  MaterialExpressionSaturate_4 {class=MaterialExpressionSaturate, inputs [None {from=MaterialExpressionAdd_1}]}
  MaterialExpressionSaturate_5 {class=MaterialExpressionSaturate, inputs [None {from=MaterialExpressionDotProduct_2}]}
  MaterialExpressionSaturate_6 {class=MaterialExpressionSaturate, inputs [None {from=MaterialExpressionAdd_5}]}
  MaterialExpressionSaturate_7 {class=MaterialExpressionSaturate, inputs [None {from=MaterialExpressionStaticSwitchParameter_6}]}
  MaterialExpressionSaturate_8 {class=MaterialExpressionSaturate, inputs [None {from=MaterialExpressionSphereMask_2}]}
  MaterialExpressionSaturate_9 {class=MaterialExpressionSaturate, inputs [None {from=MaterialExpressionOneMinus_1}]}
  MaterialExpressionScalarParameter_1 {class=MaterialExpressionScalarParameter, properties {group=B_Rocks Sides Colors, parameterName=Height}}
  MaterialExpressionScalarParameter_10 {
   class=MaterialExpressionScalarParameter
   properties {defaultValue=1024, group=B_Rocks Sides L2, parameterName=Rocks Side Scale L2, sortPriority=3}
  }
  MaterialExpressionScalarParameter_11 {
   class=MaterialExpressionScalarParameter
   properties {defaultValue=500, group=B_Rocks Sides L2 mask, parameterName=L2 Sphere Mask Radius, sortPriority=2}
  }
  MaterialExpressionScalarParameter_12 {class=MaterialExpressionScalarParameter, properties {group=B_Rocks Sides L2 mask, parameterName=L2 Sphere Mask Contrast, sortPriority=3}}
  MaterialExpressionScalarParameter_13 {class=MaterialExpressionScalarParameter, properties {defaultValue=1024, group=B_Rocks Sides, parameterName=Rocks Side Scale, sortPriority=3}}
  MaterialExpressionScalarParameter_14 {class=MaterialExpressionScalarParameter, properties {group=B_Rocks Sides L2, parameterName=Rocks Side L2 Desaturate, sortPriority=-1}}
  MaterialExpressionScalarParameter_15 {class=MaterialExpressionScalarParameter, properties {defaultValue=683.0411, group=A_Rocks Top, parameterName=RocksTopScale, sortPriority=3}}
  MaterialExpressionScalarParameter_16 {
   class=MaterialExpressionScalarParameter
   properties {defaultValue=1, group=B_Rocks Sides L2, parameterName=Rocks Side L2 N Value, sortPriority=4}
  }
  MaterialExpressionScalarParameter_17 {class=MaterialExpressionScalarParameter, properties {defaultValue=0.95, group=D_Rocks Global, parameterName=Rocks Up/Sides}}
  MaterialExpressionScalarParameter_18 {class=MaterialExpressionScalarParameter, properties {defaultValue=0.5, group=D_Rocks Global, parameterName=Rocks Up/Sides Contrast}}
  MaterialExpressionScalarParameter_19 {class=MaterialExpressionScalarParameter, properties {defaultValue=500, group=D_Rocks Global, parameterName=AO Radius}}
  MaterialExpressionScalarParameter_20 {class=MaterialExpressionScalarParameter, properties {defaultValue=75, group=D_Rocks Global, parameterName=AO Base Distance}}
  MaterialExpressionScalarParameter_21 {class=MaterialExpressionScalarParameter, properties {group=B_Rocks Sides L2 mask, parameterName=L2 Mask Contrast, sortPriority=1}}
  MaterialExpressionScalarParameter_22 {
   class=MaterialExpressionScalarParameter
   properties {defaultValue=2, desc=Details Side, group=RocksSides, parameterName=Details To Rocks Sides, sortPriority=5}
  }
  MaterialExpressionScalarParameter_23 {class=MaterialExpressionScalarParameter, properties {defaultValue=1, group=B_Rocks Sides, parameterName=Rocks Side N Value, sortPriority=4}}
  MaterialExpressionScalarParameter_24 {
   class=MaterialExpressionScalarParameter
   properties {defaultValue=0.5, group=B_Rocks Sides L2 mask, parameterName=L2 Mask Transition, sortPriority=0}
  }
  MaterialExpressionScalarParameter_25 {class=MaterialExpressionScalarParameter, properties {defaultValue=1, group=A_Rocks Top, parameterName=Rocks Top N Value, sortPriority=5}}
  MaterialExpressionScalarParameter_26 {
   class=MaterialExpressionScalarParameter
   properties {defaultValue=2, desc=Details Top, group=Rocks Top, parameterName=Details To Rocks Top, sortPriority=5}
  }
  MaterialExpressionScalarParameter_27 {class=MaterialExpressionScalarParameter, properties {group=B_Rocks Sides, parameterName=Rocks Side ProjectionContrast, sortPriority=3}}
  MaterialExpressionScalarParameter_28 {class=MaterialExpressionScalarParameter, properties {group=B_Rocks Sides L2 mask, parameterName=L2 Mask HM Min, sortPriority=5}}
  MaterialExpressionScalarParameter_29 {
   class=MaterialExpressionScalarParameter
   properties {defaultValue=1, group=B_Rocks Sides L2 mask, parameterName=L2 Mask HM Max, sortPriority=5}
  }
  MaterialExpressionScalarParameter_3 {class=MaterialExpressionScalarParameter, properties {group=Waterline, parameterName=WaterHeight Z, sortPriority=2}}
  MaterialExpressionScalarParameter_30 {class=MaterialExpressionScalarParameter, properties {defaultValue=1, desc=Related to RVT Input Temp, parameterName=L2 Rvt Mask Tweak}}
  MaterialExpressionScalarParameter_31 {class=MaterialExpressionScalarParameter, properties {defaultValue=1, group=B_Rocks Sides Colors, parameterName=Specular Level}}
  MaterialExpressionScalarParameter_32 {class=MaterialExpressionScalarParameter, properties {defaultValue=1, group=Waterline, parameterName=Waterline Opacity, sliderMax=1}}
  MaterialExpressionScalarParameter_33 {class=MaterialExpressionScalarParameter, properties {defaultValue=1, group=B_Rocks Sides, parameterName=Rocks Side Pow, sortPriority=-1}}
  MaterialExpressionScalarParameter_34 {class=MaterialExpressionScalarParameter, properties {defaultValue=0.5, group=D_Rocks Global, parameterName=Dust Spec Level}}
  MaterialExpressionScalarParameter_35 {class=MaterialExpressionScalarParameter, properties {defaultValue=1, group=C_Baked Prop, parameterName=Metal Normals Value}}
  MaterialExpressionScalarParameter_36 {class=MaterialExpressionScalarParameter, properties {defaultValue=2.5, group=C_Baked Prop, parameterName=SpecChannelModDiffuse_Mod}}
  MaterialExpressionScalarParameter_37 {class=MaterialExpressionScalarParameter, properties {defaultValue=0.5, group=C_Baked Prop, parameterName=SpecChannelModDiffuse_Desat}}
  MaterialExpressionScalarParameter_38 {class=MaterialExpressionScalarParameter, properties {defaultValue=3.25, group=C_Baked Prop, parameterName=SpecChannelModDiffuse_SpecMod}}
  MaterialExpressionScalarParameter_39 {class=MaterialExpressionScalarParameter, properties {defaultValue=-0.75, group=C_Baked Prop, parameterName=SpecChannelModDiffuse_SpecAdd}}
  MaterialExpressionScalarParameter_4 {class=MaterialExpressionScalarParameter, properties {group=Spots, parameterName=Tweak Details ProjectionContrast}}
  MaterialExpressionScalarParameter_45 {
   class=MaterialExpressionScalarParameter
   properties {defaultValue=512, group=B_Rocks Sides Colors, parameterName=RVT Rocks Side Color Mask, sortPriority=-1}
  }
  MaterialExpressionScalarParameter_5 {class=MaterialExpressionScalarParameter, properties {defaultValue=50, group=Waterline, parameterName=WaterFade, sortPriority=3}}
  MaterialExpressionScalarParameter_6 {class=MaterialExpressionScalarParameter, properties {group=A_Rocks Top, parameterName=Rocks Top Desaturate, sortPriority=-1}}
  MaterialExpressionScalarParameter_8 {class=MaterialExpressionScalarParameter, properties {group=B_Rocks Sides, parameterName=Rocks Side Desaturate, sortPriority=-1}}
  MaterialExpressionScalarParameter_9 {class=MaterialExpressionScalarParameter, properties {defaultValue=2, group=Spots, parameterName=Tweak Details Tiling}}
  MaterialExpressionSetMaterialAttributes_0 {
   class=MaterialExpressionSetMaterialAttributes
   inputs [
    MaterialAttributes {from=MaterialExpressionGetMaterialAttributes_2, output=MaterialAttributes}
    Specular {from=MaterialExpressionConstant_11, output=""}
    Roughness {from=MaterialExpressionConstant_12, output=""}
    Normal {from=MaterialExpressionLinearInterpolate_5, output=""}
    BaseColor {from=MaterialExpressionLinearInterpolate_6, output=""}
   ]
   properties {attributeSetTypes [
     #1
     #4
     #2
     #3
    ]}
  }
  MaterialExpressionSetMaterialAttributes_1 {
   class=MaterialExpressionSetMaterialAttributes
   inputs [
    MaterialAttributes {from=MaterialExpressionGetMaterialAttributes_3, output=MaterialAttributes}
    BaseColor {from=MaterialExpressionLinearInterpolate_7, output=""}
    Roughness {from=MaterialExpressionLinearInterpolate_0, output=""}
    Specular {from=MaterialExpressionConstant_2, output=""}
   ]
   properties {attributeSetTypes [#3, #4, #1]}
  }
  MaterialExpressionSetMaterialAttributes_11 {
   class=MaterialExpressionSetMaterialAttributes
   inputs [
    MaterialAttributes {from=MaterialExpressionGetMaterialAttributes_5, output=MaterialAttributes}
    Normal {from=MaterialExpressionAdd_2, output=""}
    Specular {from=MaterialExpressionGetMaterialAttributes_5, output=MaterialAttributes}
   ]
   properties {attributeSetTypes [#2, #1]}
  }
  MaterialExpressionSetMaterialAttributes_12 {
   class=MaterialExpressionSetMaterialAttributes
   inputs [
    MaterialAttributes {from="", output=""}
    BaseColor {from=MaterialExpressionVectorParameter_13, output=""}
    Roughness {from=MaterialExpressionConstant_9, output=""}
    Specular {from=MaterialExpressionConstant_10, output=""}
    Normal {from=MaterialExpressionConstant3Vector_12, output=""}
   ]
   properties {attributeSetTypes [
     #3
     #4
     #1
     #2
    ]}
  }
  MaterialExpressionSetMaterialAttributes_15 {
   class=MaterialExpressionSetMaterialAttributes
   inputs [
    MaterialAttributes {from=MaterialExpressionGetMaterialAttributes_1, output=MaterialAttributes}
    BaseColor {from=MaterialExpressionVectorParameter_16, output=""}
    Roughness {from=MaterialExpressionConstant_39, output=""}
    Specular {from=MaterialExpressionMultiply_28, output=""}
   ]
   properties {attributeSetTypes [#3, #4, #1]}
  }
  MaterialExpressionSetMaterialAttributes_16 {
   class=MaterialExpressionSetMaterialAttributes
   inputs [
    MaterialAttributes {from=MaterialExpressionGetMaterialAttributes_7, output=MaterialAttributes}
    BaseColor {from=MaterialExpressionStaticSwitchParameter_20, output=""}
    Specular {from=MaterialExpressionMultiply_23, output=""}
    Roughness {from=MaterialExpressionMultiply_29, output=""}
   ]
   properties {attributeSetTypes [#3, #1, #4]}
  }
  MaterialExpressionSetMaterialAttributes_2 {
   class=MaterialExpressionSetMaterialAttributes
   inputs [
    MaterialAttributes {from="", output=""}
    BaseColor {from=MaterialExpressionMultiply_19, output=""}
    Specular {from=MaterialExpressionTextureSampleParameter2D_1, output=R}
    Metallic {from=MaterialExpressionNamedRerouteDeclaration_19, output=""}
    Roughness {from=MaterialExpressionTextureSampleParameter2D_1, output=R}
    Normal {from=MaterialExpressionTextureSampleParameter2D_2, output=RGB}
   ]
   properties {attributeSetTypes [
     #3
     #1
     #5
     #4
     #2
    ]}
  }
  MaterialExpressionSetMaterialAttributes_3 {
   class=MaterialExpressionSetMaterialAttributes
   inputs [
    MaterialAttributes {from=MaterialExpressionGetMaterialAttributes_0, output=MaterialAttributes}
    BaseColor {from=MaterialExpressionLinearInterpolate_17, output=""}
    Specular {from=MaterialExpressionMultiply_4, output=""}
    Roughness {from=MaterialExpressionConstant_17, output=""}
    Normal {from=MaterialExpressionLinearInterpolate_16, output=""}
   ]
   properties {attributeSetTypes [
     #3
     #1
     #4
     #2
    ]}
  }
  MaterialExpressionSetMaterialAttributes_4 {
   class=MaterialExpressionSetMaterialAttributes
   inputs [
    MaterialAttributes {from="", output=""}
    BaseColor {from=MaterialExpressionMultiply_27, output=""}
    Specular {from=MaterialExpressionMaterialFunctionCall_10, output=R}
    Roughness {from=MaterialExpressionMaterialFunctionCall_10, output=R}
    Normal {from=MaterialExpressionAppendVector_7, output=""}
    Metallic {from=MaterialExpressionStaticSwitchParameter_18, output=""}
   ]
   properties {attributeSetTypes [
     #3
     #1
     #4
     #2
     #5
    ]}
  }
  MaterialExpressionSetMaterialAttributes_7 {
   class=MaterialExpressionSetMaterialAttributes
   inputs [
    MaterialAttributes {from="", output=""}
    BaseColor {from=MaterialExpressionNamedRerouteDeclaration_42, output=""}
    Specular {from=MaterialExpressionMaterialFunctionCall_9, output=R}
    Roughness {from=MaterialExpressionMaterialFunctionCall_9, output=R}
    Normal {from=MaterialExpressionStaticSwitchParameter_14, output=""}
    Metallic {from=MaterialExpressionStaticSwitchParameter_17, output=""}
   ]
   properties {attributeSetTypes [
     #3
     #1
     #4
     #2
     #5
    ]}
  }
  MaterialExpressionSetMaterialAttributes_9 {
   class=MaterialExpressionSetMaterialAttributes
   inputs [
    MaterialAttributes {from="", output=""}
    BaseColor {from=MaterialExpressionMultiply_15, output=""}
    Specular {from=MaterialExpressionMaterialFunctionCall_12, output=R}
    Roughness {from=MaterialExpressionMultiply_9, output=""}
    Normal {from=MaterialExpressionAppendVector_3, output=""}
   ]
   properties {attributeSetTypes [
     #3
     #1
     #4
     #2
    ]}
  }
  MaterialExpressionShadingPathSwitch_0 {
   class=MaterialExpressionShadingPathSwitch
   inputs [
    Default {from=MaterialExpressionDistanceFieldApproxAO_1, output=""}
    Deferred {from="", output=""}
    Forward {from=MaterialExpressionConstant_1, output=""}
    Mobile {from=MaterialExpressionConstant_1, output=""}
   ]
   properties {}
  }
  MaterialExpressionShadingPathSwitch_1 {
   class=MaterialExpressionShadingPathSwitch
   inputs [
    Default {from=MaterialExpressionStaticSwitchParameter_6, output=""}
    Deferred {from="", output=""}
    Forward {from="", output=""}
    Mobile {from=MaterialExpressionReroute_3, output=""}
   ]
   properties {}
  }
  MaterialExpressionSphereMask_0 {
   class=MaterialExpressionSphereMask
   inputs [
    A {from=MaterialExpressionWorldPosition_2, output=XYZ}
    B {from=MaterialExpressionAdd_3, output=""}
    Radius {from=MaterialExpressionScalarParameter_11, output=""}
    Hardness {from=MaterialExpressionScalarParameter_12, output=""}
   ]
   properties {}
  }
  MaterialExpressionSphereMask_2 {
   class=MaterialExpressionSphereMask
   inputs [
    A {from=MaterialExpressionVertexNormalWS_5, output=""}
    B {from=MaterialExpressionConstant3Vector_0, output=""}
    Radius {from=MaterialExpressionScalarParameter_17, output=""}
    Hardness {from=MaterialExpressionScalarParameter_18, output=""}
   ]
   properties {}
  }
  MaterialExpressionStaticBool_0 {class=MaterialExpressionStaticBool}
  MaterialExpressionStaticBool_1 {class=MaterialExpressionStaticBool}
  MaterialExpressionStaticBool_10 {class=MaterialExpressionStaticBool, properties {value=true}}
  MaterialExpressionStaticBool_2 {class=MaterialExpressionStaticBool}
  MaterialExpressionStaticBool_3 {class=MaterialExpressionStaticBool, properties {value=true}}
  MaterialExpressionStaticBool_4 {class=MaterialExpressionStaticBool}
  MaterialExpressionStaticBool_6 {class=MaterialExpressionStaticBool}
  MaterialExpressionStaticBool_7 {class=MaterialExpressionStaticBool, properties {value=true}}
  MaterialExpressionStaticBool_8 {class=MaterialExpressionStaticBool, properties {value=true}}
  MaterialExpressionStaticSwitchParameter_0 {
   class=MaterialExpressionStaticSwitchParameter
   inputs [
    True {from=MaterialExpressionBlendMaterialAttributes_4, output=""}
    False {from=MaterialExpressionNamedRerouteUsage_5, output=""}
   ]
   properties {group=Waterline, parameterName=WaterlineTrue, sortPriority=-1}
  }
  MaterialExpressionStaticSwitchParameter_1 {
   class=MaterialExpressionStaticSwitchParameter
   inputs [
    True {from=MaterialExpressionNamedRerouteUsage_52, output=""}
    False {from=MaterialExpressionQualitySwitch_4, output=""}
   ]
   properties {desc=NOTE:Disabled below EPIC, group=E_Spots, parameterName=Rocks_Top_SkipSpots, sortPriority=1}
  }
  MaterialExpressionStaticSwitchParameter_10 {
   class=MaterialExpressionStaticSwitchParameter
   inputs [
    True {from=MaterialExpressionTextureSample_1, output=RGB}
    False {from=MaterialExpressionMaterialFunctionCall_5, output=XY Spec}
   ]
   properties {group=B_Rocks Sides L2, parameterName=SideLayer2 Baked}
  }
  MaterialExpressionStaticSwitchParameter_11 {
   class=MaterialExpressionStaticSwitchParameter
   inputs [
    True {from=MaterialExpressionTextureSample_3, output=RGB}
    False {from=MaterialExpressionMaterialFunctionCall_5, output=Normal XYZ Flat Top}
   ]
   properties {group=B_Rocks Sides L2, parameterName=SideLayer2 Baked}
  }
  MaterialExpressionStaticSwitchParameter_12 {
   class=MaterialExpressionStaticSwitchParameter
   inputs [
    True {from=MaterialExpressionBlendMaterialAttributes_3, output=""}
    False {from=MaterialExpressionNamedRerouteUsage_37, output=""}
   ]
   properties {group=B_Rocks Sides L2, parameterName="Sides Layer 2 "}
  }
  MaterialExpressionStaticSwitchParameter_13 {
   class=MaterialExpressionStaticSwitchParameter
   inputs [
    True {from=MaterialExpressionMultiply_16, output=""}
    False {from=MaterialExpressionMultiply_17, output=""}
   ]
   properties {group=Spots, parameterName=UseSimpleDetailNormal}
  }
  MaterialExpressionStaticSwitchParameter_14 {
   class=MaterialExpressionStaticSwitchParameter
   inputs [
    True {from=MaterialExpressionNamedRerouteUsage_29, output=""}
    False {from=MaterialExpressionAppendVector_1, output=""}
   ]
   properties {group=Spots, parameterName=UseSimpleDetailNormal}
  }
  MaterialExpressionStaticSwitchParameter_15 {
   class=MaterialExpressionStaticSwitchParameter
   inputs [
    True {from=MaterialExpressionNamedRerouteUsage_12, output=""}
    False {from=MaterialExpressionSetMaterialAttributes_9, output=""}
   ]
   properties {group=C_Baked Prop, parameterName=Baked Prop, sortPriority=-5}
  }
  MaterialExpressionStaticSwitchParameter_16 {
   class=MaterialExpressionStaticSwitchParameter
   inputs [
    True {from=MaterialExpressionLinearInterpolate_24, output=""}
    False {from=MaterialExpressionTextureSampleParameter2D_0, output=RGB}
   ]
   properties {group=C_Baked Prop, parameterName=UseSpecChannelToModDiffuse}
  }
  MaterialExpressionStaticSwitchParameter_17 {
   class=MaterialExpressionStaticSwitchParameter
   inputs [
    True {from=MaterialExpressionMaterialFunctionCall_9, output=G}
    False {from=MaterialExpressionConstant_24, output=""}
   ]
   properties {group=D_Rocks Global, parameterName=G Channel is Metal}
  }
  MaterialExpressionStaticSwitchParameter_18 {
   class=MaterialExpressionStaticSwitchParameter
   inputs [
    True {from=MaterialExpressionMaterialFunctionCall_10, output=G}
    False {from=MaterialExpressionConstant_24, output=""}
   ]
   properties {group=D_Rocks Global, parameterName=G Channel is Metal}
  }
  MaterialExpressionStaticSwitchParameter_2 {
   class=MaterialExpressionStaticSwitchParameter
   inputs [
    True {from=MaterialExpressionStaticSwitchParameter_3, output=""}
    False {from=MaterialExpressionDivide_0, output=""}
   ]
   properties {group=B_Rocks Sides Colors, parameterName=Use Actor Space Gradient Mask}
  }
  MaterialExpressionStaticSwitchParameter_20 {
   class=MaterialExpressionStaticSwitchParameter
   inputs [
    True {from=MaterialExpressionMultiply_35, output=""}
    False {from=MaterialExpressionMultiply_52, output=""}
   ]
   properties {group=D_Rocks Global, parameterName=G Channel is Metal}
  }
  MaterialExpressionStaticSwitchParameter_21 {
   class=MaterialExpressionStaticSwitchParameter
   inputs [
    True {from=MaterialExpressionStaticSwitchParameter_22, output=""}
    False {from=MaterialExpressionScalarParameter_24, output=""}
   ]
   properties {desc=Default is spheremask in center of actor, group=B_Rocks Sides L2 mask, parameterName=L2 Mesh Space Mask, sortPriority=-1}
  }
  MaterialExpressionStaticSwitchParameter_22 {
   class=MaterialExpressionStaticSwitchParameter
   inputs [
    True {from=MaterialExpressionMultiply_25, output=""}
    False {from=MaterialExpressionSaturate_9, output=""}
   ]
   properties {
    desc=Use RVT gradient mult by AO to control transiton between layers
    group=B_Rocks Sides L2 mask
    parameterName=L2 use AO*RVT
    sortPriority=-3
   }
  }
  MaterialExpressionStaticSwitchParameter_26 {
   class=MaterialExpressionStaticSwitchParameter
   inputs [
    True {from=MaterialExpressionNamedRerouteUsage_33, output=""}
    False {from=MaterialExpressionNamedRerouteUsage_32, output=""}
   ]
   properties {group=B_Rocks Sides L2 mask, parameterName=Side Layer 2 use L1 Alpha, sortPriority=-5}
  }
  MaterialExpressionStaticSwitchParameter_3 {
   class=MaterialExpressionStaticSwitchParameter
   inputs [
    True {from=MaterialExpressionNamedRerouteUsage_15, output=""}
    False {from=MaterialExpressionComponentMask_14, output=""}
   ]
   properties {group=B_Rocks Sides Colors, parameterName=Use Side 2 Sphere Mask for color}
  }
  MaterialExpressionStaticSwitchParameter_4 {
   class=MaterialExpressionStaticSwitchParameter
   inputs [
    True {from=MaterialExpressionMultiply_18, output=""}
    False {from=MaterialExpressionScalarParameter_32, output=""}
   ]
   properties {group=Waterline, parameterName=Remove Waterline From Top?}
  }
  MaterialExpressionStaticSwitchParameter_5 {
   class=MaterialExpressionStaticSwitchParameter
   inputs [
    True {from=MaterialExpressionNamedRerouteUsage_14, output=""}
    False {from=MaterialExpressionNamedRerouteDeclaration_40, output=""}
   ]
   properties {group=C_Baked Prop, parameterName=Baked Prop, sortPriority=-5}
  }
  MaterialExpressionStaticSwitchParameter_6 {
   class=MaterialExpressionStaticSwitchParameter
   inputs [
    True {from=MaterialExpressionReroute_3, output=""}
    False {from=MaterialExpressionStaticSwitchParameter_2, output=""}
   ]
   properties {group=RVT, parameterName=SkipRVT}
  }
  MaterialExpressionStaticSwitchParameter_7 {
   class=MaterialExpressionStaticSwitchParameter
   inputs [
    True {from=MaterialExpressionMultiply_32, output=""}
    False {from=MaterialExpressionStaticSwitchParameter_13, output=""}
   ]
   properties {group=C_Baked Prop, parameterName=Baked Prop, sortPriority=-5}
  }
  MaterialExpressionStaticSwitchParameter_8 {
   class=MaterialExpressionStaticSwitchParameter
   inputs [
    True {from=MaterialExpressionLinearInterpolate_20, output=""}
    False {from=MaterialExpressionNamedRerouteUsage_16, output=""}
   ]
   properties {group=B_Rocks Sides L2, parameterName="Sides Layer 2 "}
  }
  MaterialExpressionStaticSwitchParameter_9 {
   class=MaterialExpressionStaticSwitchParameter
   inputs [
    True {from=MaterialExpressionTextureSample_0, output=RGBA}
    False {from=MaterialExpressionMaterialFunctionCall_5, output=XY Diffuse Texture}
   ]
   properties {group=B_Rocks Sides L2, parameterName=SideLayer2 Baked}
  }
  MaterialExpressionStaticSwitch_0 {
   class=MaterialExpressionStaticSwitch
   inputs [
    True {from=MaterialExpressionNamedRerouteUsage_31, output=""}
    False {from=MaterialExpressionBlendMaterialAttributes_6, output=""}
    Value {from=MaterialExpressionNamedRerouteUsage_26, output=""}
   ]
   properties {}
  }
  MaterialExpressionStaticSwitch_2 {
   class=MaterialExpressionStaticSwitch
   inputs [
    True {from=MaterialExpressionBlendMaterialAttributes_2, output=""}
    False {from=MaterialExpressionReroute_2, output=""}
    Value {from=MaterialExpressionNamedRerouteUsage_23, output=""}
   ]
   properties {desc="RVT Vs Down Vecotor (RVT for cliffs only)"}
  }
  MaterialExpressionStaticSwitch_5 {
   class=MaterialExpressionStaticSwitch
   inputs [
    True {from=MaterialExpressionNamedRerouteDeclaration_13, output=""}
    False {from=MaterialExpressionOneMinus_5, output=""}
    Value {from=MaterialExpressionNamedRerouteUsage_27, output=""}
   ]
   properties {desc="RVT Vs Down Vecotor (RVT for cliffs only)"}
  }
  MaterialExpressionStaticSwitch_7 {
   class=MaterialExpressionStaticSwitch
   inputs [
    True {from=MaterialExpressionComponentMask_18, output=""}
    False {from=MaterialExpressionLinearInterpolate_43, output=""}
    Value {from=MaterialExpressionReroute_37, output=""}
   ]
   properties {}
  }
  MaterialExpressionSubtract_0 {
   class=MaterialExpressionSubtract
   inputs [
    A {from=MaterialExpressionWorldPosition_0, output=XYZ}
    B {from=MaterialExpressionScalarParameter_3}
   ]
  }
  MaterialExpressionSubtract_6 {
   class=MaterialExpressionSubtract
   inputs [
    A {from=MaterialExpressionTransformPosition_0}
    B {from=MaterialExpressionAdd_4}
   ]
  }
  MaterialExpressionTextureObjectParameter_12 {
   class=MaterialExpressionTextureObjectParameter
   inputs [
    UVs {from="", output=""}
    Apply View MipBias {from="", output=""}
   ]
   properties {
    group=B_Rocks Sides L2
    parameterName=Normals Side 2
    samplerType=SAMPLERTYPE_Normal
    sortPriority=2
    texture=/Game/Environments/Asteria/Terrain/Textures/T_Asteria_QuarryCliff_02_N
   }
  }
  MaterialExpressionTextureObjectParameter_2 {
   class=MaterialExpressionTextureObjectParameter
   inputs [
    UVs {from="", output=""}
    Apply View MipBias {from="", output=""}
   ]
   properties {
    group=B_Rocks Sides
    parameterName=Diffuse Side
    sortPriority=0
    texture=/Game/Environments/Asteria/Terrain/Textures/T_Asteria_QuarryCliff_02_D
   }
  }
  MaterialExpressionTextureObjectParameter_3 {
   class=MaterialExpressionTextureObjectParameter
   inputs [
    UVs {from="", output=""}
    Apply View MipBias {from="", output=""}
   ]
   properties {
    group=B_Rocks Sides
    parameterName=Specular Side
    samplerType=SAMPLERTYPE_Masks
    sortPriority=1
    texture=/Game/Environments/Asteria/Terrain/Textures/T_Asteria_QuarryCliff_02_S
   }
  }
  MaterialExpressionTextureObjectParameter_4 {
   class=MaterialExpressionTextureObjectParameter
   inputs [
    UVs {from="", output=""}
    Apply View MipBias {from="", output=""}
   ]
   properties {
    group=B_Rocks Sides
    parameterName=Normals Side
    samplerType=SAMPLERTYPE_Normal
    sortPriority=2
    texture=&T_Asteria_QuarryCliff_02_N
   }
  }
  MaterialExpressionTextureObjectParameter_5 {
   class=MaterialExpressionTextureObjectParameter
   inputs [
    UVs {from="", output=""}
    Apply View MipBias {from="", output=""}
   ]
   properties {group=A_Rocks Top, parameterName=Diffuse Top, sortPriority=0, texture=/Game/Environments/Apollo/Sets/Terrain/Textures/T_Apollo_Cliff_D_02}
  }
  MaterialExpressionTextureObjectParameter_6 {
   class=MaterialExpressionTextureObjectParameter
   inputs [
    UVs {from="", output=""}
    Apply View MipBias {from="", output=""}
   ]
   properties {
    group=A_Rocks Top
    parameterName=Specular Top
    samplerType=SAMPLERTYPE_Masks
    sortPriority=1
    texture=/Game/Environments/Apollo/Sets/Terrain/Textures/T_Apollo_Cliff_S_02
   }
  }
  MaterialExpressionTextureObjectParameter_7 {
   class=MaterialExpressionTextureObjectParameter
   inputs [
    UVs {from="", output=""}
    Apply View MipBias {from="", output=""}
   ]
   properties {
    group=A_Rocks Top
    parameterName=Normals Top
    samplerType=SAMPLERTYPE_Normal
    sortPriority=2
    texture=/Game/Environments/Apollo/Sets/Terrain/Textures/T_Apollo_Cliff_N_02
   }
  }
  MaterialExpressionTextureObjectParameter_8 {
   class=MaterialExpressionTextureObjectParameter
   inputs [
    UVs {from="", output=""}
    Apply View MipBias {from="", output=""}
   ]
   properties {
    group=B_Rocks Sides L2
    parameterName=Diffuse Side 2
    sortPriority=0
    texture=&T_Asteria_QuarryCliff_02_D
   }
  }
  MaterialExpressionTextureObjectParameter_9 {
   class=MaterialExpressionTextureObjectParameter
   inputs [
    UVs {from="", output=""}
    Apply View MipBias {from="", output=""}
   ]
   properties {
    group=B_Rocks Sides L2
    parameterName=Specular Side 2
    samplerType=SAMPLERTYPE_Masks
    sortPriority=1
    texture=&T_Asteria_QuarryCliff_02_S
   }
  }
  MaterialExpressionTextureObject_0 {
   class=MaterialExpressionTextureObject
   inputs []
   properties {texture=/Game/Environments/Asteria/Terrain/Textures/T_Asteria_WindBlownSnow_01_D}
  }
  MaterialExpressionTextureObject_6 {
   class=MaterialExpressionTextureObject
   inputs []
   properties {samplerType=SAMPLERTYPE_Normal, texture=/Game/Environments/Asteria/Rocks/Textures/T_RockDetails2_N}
  }
  MaterialExpressionTextureObject_7 {
   class=MaterialExpressionTextureObject
   inputs []
   properties {samplerType=SAMPLERTYPE_Masks, texture=/Game/Environments/Asteria/Rocks/Textures/T_RockDetails2_M}
  }
  MaterialExpressionTextureSampleParameter2D_0 {
   class=MaterialExpressionTextureSampleParameter2D
   inputs [
    UVs {from="", output=""}
    Apply View MipBias {from="", output=""}
   ]
   properties {group=C_Baked Prop, parameterName=Diffuse, sortPriority=0, texture=&T_Apollo_Cliff_D_02}
  }
  MaterialExpressionTextureSampleParameter2D_1 {
   class=MaterialExpressionTextureSampleParameter2D
   inputs [
    UVs {from="", output=""}
    Apply View MipBias {from="", output=""}
   ]
   properties {
    group=C_Baked Prop
    parameterName=Specular
    samplerType=SAMPLERTYPE_Masks
    sortPriority=1
    texture=&T_Apollo_Cliff_S_02
   }
  }
  MaterialExpressionTextureSampleParameter2D_2 {
   class=MaterialExpressionTextureSampleParameter2D
   inputs [
    UVs {from="", output=""}
    Apply View MipBias {from="", output=""}
   ]
   properties {
    group=C_Baked Prop
    parameterName=Normals
    samplerType=SAMPLERTYPE_Normal
    sortPriority=2
    texture=&T_Apollo_Cliff_N_02
   }
  }
  MaterialExpressionTextureSample_0 {
   class=MaterialExpressionTextureSample
   inputs [
    UVs {from=""}
    Tex {from=MaterialExpressionTextureObjectParameter_8}
    Apply View MipBias {from=""}
   ]
  }
  MaterialExpressionTextureSample_1 {
   class=MaterialExpressionTextureSample
   inputs [
    UVs {from=""}
    Tex {from=MaterialExpressionTextureObjectParameter_9}
    Apply View MipBias {from=""}
   ]
  }
  MaterialExpressionTextureSample_2 {
   class=MaterialExpressionTextureSample
   inputs [
    UVs {from=MaterialExpressionComponentMask_28}
    Tex {from=""}
    Apply View MipBias {from=""}
   ]
   properties {samplerType=SAMPLERTYPE_Masks, texture=&T_RockDetails2_M}
  }
  MaterialExpressionTextureSample_3 {
   class=MaterialExpressionTextureSample
   inputs [
    UVs {from=""}
    Tex {from=MaterialExpressionTextureObjectParameter_12}
    Apply View MipBias {from=""}
   ]
  }
  MaterialExpressionTransformPosition_0 {
   class=MaterialExpressionTransformPosition
   inputs [None {from=MaterialExpressionWorldPosition_8, output=XYZ}]
   properties {transformSourceType=TRANSFORMPOSSOURCE_World}
  }
  MaterialExpressionVectorParameter_0 {
   class=MaterialExpressionVectorParameter
   properties {
    group=Waterline
    parameterName=UnderWater
    sortPriority=1
    channelNames {a=Blend to Multiply, b="", g="", r=""}
    defaultValue=<rgba 0.474 0.3844 0.1469 1>
   }
  }
  MaterialExpressionVectorParameter_1 {
   class=MaterialExpressionVectorParameter
   properties {
    group=Waterline
    parameterName=Waterline color
    sortPriority=0
    channelNames {a=Blend to Multiply, b="", g="", r=""}
    defaultValue=<rgba 0.7483 0.5324 0.2949 0.75>
   }
  }
  MaterialExpressionVectorParameter_13 {
   class=MaterialExpressionVectorParameter
   properties {
    group=E_Spots
    parameterName=Spots Color
    sortPriority=0
    channelNames {a=Spots Value, b="", g="", r=""}
    defaultValue=<rgba 1 0.8654 0.5816 1>
   }
  }
  MaterialExpressionVectorParameter_16 {
   class=MaterialExpressionVectorParameter
   properties {group=D_Rocks Global, parameterName=Dust, channelNames {a=Dust Ammount, b="", g="", r=""}, defaultValue=<rgba 0.224 0.1363 0.0678 1>}
  }
  MaterialExpressionVectorParameter_17 {
   class=MaterialExpressionVectorParameter
   properties {
    group=B_Rocks Sides Colors
    parameterName=Rocks Tone Side Bottom
    sortPriority=-2
    channelNames {a=Roughness Tweak, b="", g="", r=""}
    defaultValue=<rgba 0.9484 1 0.9831 1>
   }
  }
  MaterialExpressionVectorParameter_2 {
   class=MaterialExpressionVectorParameter
   properties {
    group=C_Baked Prop
    parameterName=Diffuse Tone Up
    sortPriority=-1
    channelNames {a=Blend To Lerp, b="", g="", r=""}
    defaultValue=<rgba 1 1 1 1>
   }
  }
  MaterialExpressionVectorParameter_3 {
   class=MaterialExpressionVectorParameter
   properties {
    group=C_Baked Prop
    parameterName=Diffuse Tone Sides
    sortPriority=-1
    channelNames {a=Blend To Lerp, b="", g="", r=""}
    defaultValue=<rgba 1 1 1 1>
   }
  }
  MaterialExpressionVectorParameter_4 {
   class=MaterialExpressionVectorParameter
   properties {group=B_Rocks Sides, parameterName=Color UpDown Sides, defaultValue=<rgba 0.1302 0.1302 0.1302 0>}
  }
  MaterialExpressionVectorParameter_5 {
   class=MaterialExpressionVectorParameter
   properties {
    desc="Based on actor location, so can be set for actor and reused when actor moved. Think about it as pivot you moving to set center of spehere"
    group=B_Rocks Sides L2 mask
    parameterName=L2 Sphere Location Tweak
    sortPriority=6
   }
  }
  MaterialExpressionVectorParameter_6 {class=MaterialExpressionVectorParameter, properties {parameterName=Rocks Side L2 Tone, defaultValue=<rgba 1 1 1 1>}}
  MaterialExpressionVectorParameter_7 {
   class=MaterialExpressionVectorParameter
   properties {
    group=B_Rocks Sides Colors
    parameterName=Rocks Tone Side
    sortPriority=-3
    channelNames {a=Roughness Tweak, b="", g="", r=""}
    defaultValue=<rgba 0.5955 0.5955 0.5955 1>
   }
  }
  MaterialExpressionVectorParameter_8 {
   class=MaterialExpressionVectorParameter
   properties {
    group=A_Rocks Top
    parameterName=Rocks Tone Top
    sortPriority=-1
    channelNames {a=Roughness Tweal, b="", g="", r=""}
    defaultValue=<rgba 0.5 0.5122 0.4741 1>
   }
  }
  MaterialExpressionVertexNormalWS_0 {class=MaterialExpressionVertexNormalWS, inputs [], properties {}}
  MaterialExpressionVertexNormalWS_1 {class=MaterialExpressionVertexNormalWS, inputs [], properties {}}
  MaterialExpressionVertexNormalWS_10 {class=MaterialExpressionVertexNormalWS, inputs [], properties {}}
  MaterialExpressionVertexNormalWS_11 {class=MaterialExpressionVertexNormalWS, inputs [], properties {}}
  MaterialExpressionVertexNormalWS_3 {class=MaterialExpressionVertexNormalWS, inputs [], properties {}}
  MaterialExpressionVertexNormalWS_5 {class=MaterialExpressionVertexNormalWS, inputs [], properties {}}
  MaterialExpressionWorldPosition_0 {class=MaterialExpressionWorldPosition, inputs [], properties {}}
  MaterialExpressionWorldPosition_1 {class=MaterialExpressionWorldPosition, inputs [], properties {}}
  MaterialExpressionWorldPosition_2 {class=MaterialExpressionWorldPosition, inputs [], properties {}}
  MaterialExpressionWorldPosition_8 {class=MaterialExpressionWorldPosition, inputs [], properties {worldPositionShaderOffset=WPT_ExcludeAllShaderOffsets}}
 }}
