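# Material function that projects a mask texture, suited to stains or drips, onto mesh sides in bounding-box space, with optionally flipped, scaled, offset and clamped Z.
# Two side-planar samples use tiled horizontal coordinates and blend their R and G channels by contrast-adjusted normal XY, instance-space by default.
# The blend's divisor adds the R weight twice rather than R plus G.
# A second output gives a separately scaled, saturated Z gradient.
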MaterialFunction /Game/Environments/Asteria/Rocks/Materials/MaterialFunctions/MF_XYPlanarWithTopScale
graph {nodes [
  n1 = MaterialFunctionCall() {materialFunction=&BoundingBoxBased_0-1_UVW}
  n2 = FunctionInput(Preview=Constant() {r=0.75}) {bUsePreviewValueAsDefault=true, inputName=XY Tiling, inputType=FunctionInput_Scalar}
  n3 = StaticSwitch(
   True=OneMinus(None=@n1.B)
   False=@n1.B
   Value=FunctionInput(Preview=StaticBool() {value=true}) {bUsePreviewValueAsDefault=true, inputName=OneMinus Z Pre Scale, inputType=FunctionInput_StaticBool}
  )
  n4 = Divide(A=@n3, B=FunctionInput(Preview=Constant() {r=0.25}) {bUsePreviewValueAsDefault=true, inputName=Z Scale, inputType=FunctionInput_Scalar})
  n5 = StaticBool()
  n6 = Subtract(
   A=StaticSwitch(
    True=OneMinus(None=@n4)
    False=@n4
    Value=FunctionInput(Preview=@n5) {bUsePreviewValueAsDefault=true, inputName=OneMinus Z Post Scale, inputType=FunctionInput_StaticBool}
   )
   B=FunctionInput(Preview=Constant()) {bUsePreviewValueAsDefault=true, inputName=Z Offset, inputType=FunctionInput_Scalar}
  )
  n7 = Clamp(None=@n6) {maxDefault=0.98}
  n8 = NamedRerouteUsage()
  n9 = TextureSample(
   UVs=MaterialFunctionCall(X=Multiply(A=@n1.G, B=@n2), Y=@n7) {materialFunction=&MakeFloat2}.Result
   Tex=@n8
  ) {samplerType=SAMPLERTYPE_LinearColor, texture=/Game/UI/Foundation/Environment/Shell/Textures/T_2240_Drips_Mask}
  n10 = TextureSample(
   UVs=MaterialFunctionCall(X=Multiply(A=@n1.R, B=@n2), Y=@n7) {materialFunction=&MakeFloat2}.Result
   Tex=@n8
  ) {samplerType=SAMPLERTYPE_LinearColor, texture=&T_2240_Drips_Mask}
  n11 = VertexNormalWS()
  n12 = ComponentMask(None=FunctionInput(Preview=Transform(None=@n11) {transformSourceType=TRANSFORMSOURCE_World, transformType=TRANSFORM_Instance}) {bUsePreviewValueAsDefault=true, inputName=Custom Local Normal}) {g=true, r=true}
  n13 = Constant() {r=0.2}
  n14 = Power(
   Base=Abs(None=Normalize(VectorInput=@n12))
   Exp=Divide(
    A=Constant() {r=1}
    B=FunctionInput(Preview=@n13) {bUsePreviewValueAsDefault=true, inputName=XY Blend Contrast, inputType=FunctionInput_Scalar}
   )
  )
  n15 = DotProduct(A=MaterialFunctionCall(X=@n9.R, Y=@n10.R) {materialFunction=&MakeFloat2}.Result, B=@n14)
  n16 = DotProduct(A=MaterialFunctionCall(X=@n9.G, Y=@n10.G) {materialFunction=&MakeFloat2}.Result, B=@n14)
  n17 = MaterialFunctionCall(Float2=@n14) {materialFunction=&BreakOutFloat2Components}
  FunctionOutput(None=Divide(A=AppendVector(A=@n15, B=@n16), B=Add(A=@n17.R, B=@n17.R)))
  n18 = Divide(
   A=@n3
   B=FunctionInput(Preview=Constant() {r=0.25}) {bUsePreviewValueAsDefault=true, inputName=Z Scale Gradient, inputType=FunctionInput_Scalar}
  )
  n19 = OneMinus(None=@n18)
  n20 = FunctionInput(Preview=StaticBool() {value=true}) {bUsePreviewValueAsDefault=true, inputName=OneMinus Z Post Scale Gradient, inputType=FunctionInput_StaticBool}
  FunctionOutput(None=Saturate(None=StaticSwitch(True=@n19, False=@n18, Value=@n20))) {outputName=Z Gradient}
  NamedRerouteDeclaration(None=FunctionInput(Preview=TextureObject() {samplerType=SAMPLERTYPE_Masks, texture=/Game/Environments/Asteria/Sets/NeonCity/Techart/Textures/T_StainA_Masks}) {bUsePreviewValueAsDefault=true, description=Texture Object, inputName=Texture Object, inputType=FunctionInput_Texture2D}) {name=Tex, nodeColor=<rgba 0.5469 0 1 1>}
  Saturate(None=@n6)
  ScalarParameter() {defaultValue=0.25, parameterName=VerticalSquash}
  ScalarParameter() {parameterName=VerticalOffset}
  ScalarParameter() {defaultValue=0.2, parameterName=BlendContrast}
 ]}
