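# Master rock material for Asteria environments, supporting Nanite and instanced meshes, built on MF_Rock2022Base with world-space normals and jungle moss.
# Static switches add fallback normal, AO and curvature via NaniteReplace and QualitySwitch, plus masked landscape top layers like grass, forest or snow.
# Landscape runtime virtual textures blend colors and normals by height difference, with optional melting-snow subsurface blending.
# Final switches add mesh normals, opacity hiding, destructible damage, Curie fallback and creative move-tool effects.
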Material /Game/Environments/Asteria/Rocks/Materials/M_Rocks_2022
graph {
 properties {
  bTangentSpaceNormal=false
  bUseMaterialAttributes=true
  bUsedWithInstancedStaticMeshes=true
  bUsedWithNanite=true
  pixelDepthOffsetMode=PDOM_Legacy
  shadingModel=MSM_FromMaterialExpression
  usedShadingModels=Default Lit | Subsurface
 }
 nodes [
  n1 = MaterialFunctionCall!(
   RVTH=NamedRerouteUsage()
   "Disable Top Texures ( Use mat layer)"=StaticSwitchParameter(True=StaticBool() {value=true}, False=StaticBool()) {desc=Enable to access Terrain layers for top surfaces, group=TopMaterials, parameterName=Add Top Material}
  ) {materialFunction=/Game/Environments/Asteria/Rocks/Materials/MaterialFunctions/MF_Rock2022Base}
  MaterialFunctionCall!(
   AO=NamedRerouteDeclaration(None=@n1.AO) {name=AO, nodeColor=<rgba 1 0 0.9844 1>}
   RVT_Height=NamedRerouteUsage()
   For Buildings
   For Moss=@n1.Result
  ) {materialFunction=/Game/Environments/Asteria/Rocks/Materials/MaterialFunctions/MF_JungleBiome}
  n2 = TextureCoordinate() {coordinateIndex=2}
  n3 = TextureCoordinate()
  n4 = TextureSampleParameter2D(UVs=NamedRerouteDeclaration(None=StaticSwitchParameter(True=@n2, False=@n3) {group=Fallback, parameterName="Use UV2 for FallBack "}) {name=UseUV2ForFallback, nodeColor=<rgba 0.0129 1 0.2737 1>}) {group=Fallback, parameterName=Fallback Normal, samplerType=SAMPLERTYPE_Normal, texture=&FlatNormalBlue}
  n5 = Transform(None=MaterialFunctionCall!(UVs=NamedRerouteUsage(), Normal=@n4.RGB) {materialFunction=/Engine/Functions/Engine_MaterialFunctions01/Texturing/DeriveTangentBasis}.World Space Normal) {transformSourceType=TRANSFORMSOURCE_World, transformType=TRANSFORM_Tangent}
  n6 = Normalize(VectorInput=MaterialFunctionCall(BaseNormal=NamedRerouteUsage(), AdditionalNormal=@n5) {materialFunction=&BlendAngleCorrectedNormals}.Result)
  n7 = Constant3Vector() {constant=<rgba 1 1 0 0>}
  n8 = QualitySwitch(Default=@n6, Low=Add(A=Multiply(A=@n5, B=@n7), B=NamedRerouteUsage()))
  n9 = NaniteReplace!(Default=@n8, Nanite=NamedRerouteUsage())
  n10 = NamedRerouteUsage()
  n11 = Constant() {r=0.75}
  n12 = NamedRerouteUsage()
  n13 = TextureSampleParameter2D(UVs=NamedRerouteUsage()) {group=Fallback, parameterName=Fallback AO, texture=/Game/Global/Textures/Default/Blanks/T_Blank_LegalWhite_sRGBColor}
  n14 = OneMinus(None=@n13.R)
  n15 = OneMinus(None=@n13.RGB)
  n16 = StaticSwitchParameter(
   True=LinearInterpolate(
    A=@n10
    B=Multiply(A=@n11, B=@n12)
    Alpha=StaticSwitchParameter(True=@n14, False=@n15) {group=Fallback, parameterName="Use Channel Packed AO (R)"}
   ) {constAlpha=0.25}
   False=@n10
  ) {group=Fallback, parameterName=Fallback AO}
  n17 = QualitySwitch(Default=@n16, Low=NamedRerouteUsage(), High=@n16, Medium=@n16, Epic=@n16)
  n18 = StaticSwitchParameter(True=NamedRerouteUsage(), False=NaniteReplace(Default=@n17, Nanite=NamedRerouteUsage())) {group=Fallback, parameterName="Use Curvature (G channel)"}
  n19 = StaticSwitchParameter(
   True=SetMaterialAttributes!(
    MaterialAttributes=NamedRerouteUsage()
    Normal=StaticSwitchParameter(True=@n8, False=@n9) {desc=Normal/AO, group=Fallback, parameterName=Use Fallback Textures on Nanite Platforms}
    Base Color=StaticSwitchParameter(True=@n17, False=@n18) {desc=Normal/AO, group=Fallback, parameterName=Use Fallback Textures on Nanite Platforms}
   ) {attributeSetTypes [#1, #2]}
   False=NamedRerouteUsage()
  ) {group=Fallback, parameterName=Use FallBack}
  n20 = GetMaterialAttributes(None=@n19) {attributeGetTypes [#1]}
  n21 = SetMaterialAttributes!(MaterialAttributes=@n20.MaterialAttributes, Normal=Transform(None=@n20.Normal)) {attributeSetTypes [#1]}
  n22 = MaterialFunctionCall!(Grass_UseCustomSlopeMask, Grass_CustomSlopeMask) {materialFunction=/Game/Athena/Helios/Environments/Landscape/Materials/MaterialFunctions_Utilities/MF_Helios_StaticMeshLandscapeLayerMatch}
  n23 = VectorParameter() {group=TopMaterials, parameterName=VertexColor Mask for Grass, defaultValue=<rgba 0 1 0 0>}
  n24 = VertexColor()
  n25 = OneMinus(None=StaticSwitchParameter(True=DotProduct(A=@n23.RGB, B=@n24), False=@n24.A) {group=TopMaterials, parameterName=Use VertexColor for Grass Blend})
  n26 = Saturate(None=Multiply(A=ScalarParameter() {defaultValue=1, group=TopMaterials, parameterName=HeliosGrassVertMaskMod}, B=@n25))
  n27 = NamedRerouteUsage()
  n28 = NamedRerouteUsage()
  n29 = MaterialFunctionCall() {materialFunction=/Game/Athena/Asteria/Environments/Landscape/Materials/MaterialFunctions/MobileLayers/MF_Mobile_Asteria_Snow}
  n30 = ComponentMask(None=MaterialFunctionCall!(UseLandscapeCoords=StaticBool()) {materialFunction=/Game/Athena/Asteria/Environments/Landscape/Materials/MaterialFunctions/MF_AsteriaTerrainMasks}.BiomeMaskRGBA) {r=true}
  n31 = MaterialFunctionCall() {materialFunction=/Game/Athena/Asteria/Environments/Landscape/Materials/MaterialFunctions/MF_AsteriaLandscapeLayers}
  n32 = StaticSwitchParameter(True=@n31.Snow, False=@n31.Snow) {group=TopMaterials, parameterName=Snow}
  n33 = StaticSwitchParameter(
   True=MaterialFunctionCall!(BiomeMask-RGBA=Constant4Vector() {constant=<rgba 0 0 1 0>}.RGBA) {materialFunction=/Game/Athena/Asteria/Environments/Landscape/Materials/MaterialFunctions/MF_Asteria_Biome_Forest}.Material
   False=StaticSwitchParameter(
    True=StaticSwitchParameter(True=@n32, False=@n31.Gravel) {desc=Enable Sand to Access Asphalt, group=TopMaterials, parameterName=Sand}
    False=@n31.CommonGrass
   ) {desc=Enable gravel to access Sand and Asphalt, group=TopMaterials, parameterName=Gravel}
  ) {desc=Replace Grass with Forest, group=TopMaterials, parameterName=Forest}
  n34 = StaticSwitchParameter(
   True=NamedRerouteDeclaration(None=MaterialFunctionCall!(
     Material 1=@n27
     Height Texture 1=@n28
     Material 2=@n29.Result
     Height Texture 2=@n29.Result
     Transistion Phase=@n30
     UseContrast
     Contrast
     AddLayerMask
    ) {materialFunction=/Game/Athena/Apollo/Environments/Landscape/Materials/Biome/CommonMaterials/MF_BlendLandscapeLayers}.BlendedMaterial) {name=TopMaterialWithSnowBiomeBlend, nodeColor=<rgba 1 0 0.375 1>}
   False=NamedRerouteDeclaration(None=StaticSwitchParameter(True=@n31.JungleFloor, False=@n33) {parameterName=Jungle}) {name=ToSnowBiomeBlend, nodeColor=<rgba 0 0.3437 1 1>}
  ) {group=SnowBiomeBlendNonRVT, parameterName=UseSnowBiomeBlend}
  n35 = NamedRerouteUsage()
  n36 = Add(A=@n25, B=LinearInterpolate(Alpha=ComponentMask(None=@n35) {r=true}) {constB=0.1})
  n37 = NamedRerouteUsage()
  n38 = StaticSwitchParameter(True=Multiply(A=Saturate(None=@n36), B=@n37), False=@n37) {group=TopMaterials, parameterName=Use Vert Colors to Mask Grass}
  n39 = GetMaterialAttributes(None=StaticSwitchParameter(True=BlendMaterialAttributes(A=@n19, B=@n34, Alpha=@n38), False=@n19) {desc=Enable to access Terrain layers for top surfaces, group=TopMaterials, parameterName=Add Top Material}) {attributeGetTypes [#1]}
  n40 = Transform(None=@n39.Normal)
  NamedRerouteDeclaration(None=StaticSwitchParameter(
    True=BlendMaterialAttributes(A=@n21, B=@n22.GrassLayer, Alpha=@n26)
    False=SetMaterialAttributes(MaterialAttributes=@n39.MaterialAttributes, Normal=@n40) {attributeSetTypes [#1]}
   ) {group=TopMaterials, parameterName=UseHeliosGrass}) {name=PreRVT, nodeColor=<rgba 0.4766 1 0 1>}
  NamedRerouteDeclaration(None=@n1.Details XYZ RGB Mask) {name=Details XYZ RGB Mask, nodeColor=<rgba 0 0.3727 1 1>}
  NamedRerouteDeclaration(None=@n1.Dust) {name=Dust, nodeColor=<rgba 0.0129 1 0.2737 1>}
  NamedRerouteDeclaration(None=@n31.Grass H) {name=GrassHeight, nodeColor=<rgba 1 0.3281 0 1>}
  n41 = GetMaterialAttributes(None=NamedRerouteDeclaration(None=MaterialFunctionCall!(In=@n1.Result) {materialFunction=/Game/Environments/Asteria/Foliage/Trees/MasterMaterials/MaterialFunctions/MF_JungleMoss_World}.Result) {name=Rocks Base, nodeColor=<rgba 1 0.0703 0 1>}) {attributeGetTypes [#1, #2]}
  NamedRerouteDeclaration(None=@n41.Normal) {name=Rocks Base_Normals, nodeColor=<rgba 1 0.0703 0 1>}
  NamedRerouteDeclaration(None=@n41.Base Color) {name=Rocks Base_Basecolor, nodeColor=<rgba 1 0.0703 0 1>}
  n42 = Multiply(A=ComponentMask(None=VertexNormalWS()) {b=true}, B=@n1.Up-Side MaskBetter)
  NamedRerouteDeclaration(None=StaticSwitchParameter(True=Saturate(None=@n42), False=@n1.Up-Side MaskBetter) {defaultValue=true, group=TopMaterials, parameterName=MaskGrassByZ}) {name=Grassy Rock, nodeColor=<rgba 0 0.9531 1 1>}
  n43 = VectorParameter() {parameterName=Curvature Color, defaultValue=<rgba 0.4115 0.4115 0.4115 1>}
  n44 = NamedRerouteUsage()
  n45 = Saturate(None=Multiply(
    A=ScalarParameter() {defaultValue=1, parameterName="Add Soft Curvature "}
    B=TextureSampleParameter2D(UVs=@n44) {group=Fallback, parameterName=Fallback AO, texture=&T_Blank_LegalWhite_sRGBColor}.G
   ))
  n46 = Saturate(None=Multiply(A=ScalarParameter() {defaultValue=1, parameterName="Add Soft Curvature "}, B=VertexColor().G))
  n47 = NaniteReplace(Default=@n45, Nanite=@n46)
  NamedRerouteDeclaration(None=Add(A=NamedRerouteUsage(), B=LinearInterpolate(B=@n43.RGB, Alpha=@n47))) {name=Curvature, nodeColor=<rgba 0 1 0.6719 1>}
  n48 = NamedRerouteUsage()
  n49 = WorldPosition() {worldPositionShaderOffset=WPT_ExcludeAllShaderOffsets}
  n50 = NamedRerouteUsage()
  n51 = ComponentMask(None=VertexNormalWS()) {g=true, r=true}
  n52 = Constant()
  n53 = Multiply(
   A=@n50
   B=MaterialFunctionCall(Vector=AppendVector(A=@n51, B=@n52), Default=Constant3Vector() {constant=<rgba 0 0 1 0>}) {materialFunction=&SafeNormalize}.Result
  )
  n54 = Divide(A=@n50) {constB=50}
  n55 = RuntimeVirtualTextureSampleParameter!(UVs, World Position=VertexInterpolator(VS=Subtract(A=@n49.XYZ, B=@n53)).PS, Mip Bias=LinearInterpolate(Alpha=Saturate(None=@n54)) {constB=2}) {
   group=RVT
   materialType=BaseColor_Normal_Specular_YCoCg
   mipValueMode=RVTMVM_MipBias
   parameterName=RVT
   virtualTexture=/Game/Athena/Apollo/Maps/Landscape/VT/RVT_Apollo_Landscape
  }
  n56 = WorldPosition()
  n57 = RuntimeVirtualTextureSampleParameter!(UVs, World Position) {
   group=RVT
   materialType=WorldHeight
   parameterName=RVTHeight
   virtualTexture=/Game/Athena/Apollo/Maps/Landscape/VT/RVT_Apollo_Landscape_Height
  }
  n58 = Subtract(
   A=Add(A=ScalarParameter() {group=RVT, parameterName=Offset RVT WP}, B=ComponentMask(None=@n56.XYZ) {b=true})
   B=NamedRerouteDeclaration(None=@n57.WorldHeight) {name=RVT Height as is, nodeColor=<rgba 0 0.4141 1 1>}
  )
  n59 = NamedRerouteDeclaration(None=MaterialFunctionCall!(NoRVT=Constant(), RVT=@n58) {materialFunction=/Game/Packages/Fortress_SharedMaterials/Base_Material_Functions/FN_RVT_Switch_Scalar}.Result) {name=RVT Height, nodeColor=<rgba 0 0.8359 1 1>}
  n60 = ScalarParameter() {
   defaultValue=20
   desc="Controls the blending of normals on sides of meshes, but not UP vector surfaces"
   group=RVT
   parameterName=Blend RVT Normals
  }
  n61 = ScalarParameter() {
   defaultValue=60
   desc="Used to control how far colors will blend with the rvt, as well sets blending distance of normals for Up looking surfaces"
   group=RVT
   parameterName=Blend RVT
  }
  n62 = NamedRerouteUsage()
  n63 = LinearInterpolate(A=@n60, B=@n61, Alpha=@n62)
  n64 = BlendMaterialAttributes(
   A=SetMaterialAttributes(MaterialAttributes=@n48, Normal=NamedRerouteDeclaration(None=@n55.Normal) {name=RVT N, nodeColor=<rgba 0.8047 1 0 1>}) {attributeSetTypes [#1]}
   B=@n48
   Alpha=Saturate(None=Divide(A=@n59, B=@n63))
  ) {vertexAttributeBlendType=UseA}
  n65 = SetMaterialAttributes!(
   MaterialAttributes=@n64
   Base Color=NamedRerouteDeclaration(None=@n55.BaseColor) {name=RVT Base Color, nodeColor=<rgba 0.6875 0 1 1>}
   Specular=NamedRerouteDeclaration(None=@n55.Specular) {name=RVT Specular, nodeColor=<rgba 1 0 0.8203 1>}
   Roughness=NamedRerouteDeclaration(None=@n55.Roughness) {name=RVT Rough, nodeColor=<rgba 0 1 0.7422 1>}
   Emissive Color=Constant()
  ) {attributeSetTypes [
    #2
    #3
    #4
    #5
   ]}
  n66 = Divide(A=VertexColor(), B=VectorParameter() {group=RVT, parameterName=Vertex Mask Channel Mask, defaultValue=<rgba 0 0 0 1>}.RGB)
  n67 = OneMinus(None=@n66)
  n68 = Divide(A=@n61, B=ScalarParameter() {defaultValue=4, group=RVT, parameterName=Divide Side RVT Distance}) {constB=4}
  n69 = NamedRerouteDeclaration(None=Divide(A=@n58, B=LinearInterpolate(A=@n68, B=@n61, Alpha=@n62))) {name=RVT Blending, nodeColor=<rgba 0.9922 1 0 1>}
  n70 = StaticSwitchParameter(True=Multiply(A=@n69, B=StaticSwitchParameter(True=@n67, False=@n66) {group=RVT, parameterName=Invert Vertex Color Mask On RVT}), False=@n69) {group=RVT, parameterName=Use Vertex Color Mask On RVT}
  n71 = BlendMaterialAttributes(
   A=NamedRerouteDeclaration(None=@n65) {name=RVT Colors, nodeColor=<rgba 0.5937 1 0 1>}
   B=@n64
   Alpha=Saturate(None=NamedRerouteDeclaration(None=@n70) {name=RVTMask, nodeColor=<rgba 0 0.9531 1 1>})
  ) {vertexAttributeBlendType=UseA}
  NamedRerouteDeclaration(None=MaterialFunctionCall(
    NoRVT=SetMaterialAttributes!(MaterialAttributes)
    RVT=SetMaterialAttributes!(MaterialAttributes=@n71, Base Color) {attributeSetTypes [#2]}
   ) {materialFunction=&FN_RVT_Switch_Material}.Result) {name=RVT On, nodeColor=<rgba 0 0.8594 1 1>}
  Reroute(None=Multiply(A=@n69, B=NamedRerouteUsage())) {desc=removed to get rid of RVT blending over dirt in hades}
  Reroute(None=@n1.Result)
  n72 = Transform(None=Constant3Vector() {constant=<rgba 0 0 1 0>})
  RuntimeVirtualTextureOutput!(
   BaseColor
   Specular
   Roughness
   Normal
   WorldHeight=MaterialFunctionCall(NoRVT=Constant(), RVT=ComponentMask(None=@n72) {b=true}) {materialFunction=&FN_RVT_Switch_Scalar}.Result
   Opacity
   Mask
   Displacement
   Mask4
  )
  Saturate(None=ComponentMask(None=VertexNormalWS()) {b=true})
  n73 = NamedRerouteUsage()
  n74 = MaterialFunctionCall!(
   Z Offset
   XY Tiling=Constant() {r=1.6666}
   Z Scale=Constant() {r=1.25}
   OneMinus Z Pre Scale
   OneMinus Z Post Scale
   Texture Object=TextureObject() {samplerType=SAMPLERTYPE_Masks, texture=/Game/Environments/Asteria/Sets/NeonCity/Techart/Textures/T_StainA_Masks}
   Custom Local Normal
   XY Blend Contrast
   Z Scale Gradient=Constant() {r=0.75}
   OneMinus Z Post Scale Gradient
  ) {materialFunction=/Game/Environments/Asteria/Rocks/Materials/MaterialFunctions/MF_XYPlanarWithTopScale}
  n75 = ComponentMask(None=@n74.Result) {r=true}
  n76 = MaterialFunctionCall(In=Multiply(A=@n74.Z Gradient, B=Multiply(A=@n75) {constB=3}), Contrast=Constant() {r=2}) {materialFunction=&CheapContrast}
  n77 = ComponentMask(None=@n74.Result) {g=true}
  n78 = OneMinus(None=Multiply(A=@n76.Result, B=Power(Base=@n77) {constExponent=3}))
  n79 = ScalarParameter() {defaultValue=0.2, parameterName=Wet Melting Snow Roughness}
  n80 = Constant() {r=1.1}
  n81 = NamedRerouteUsage()
  n82 = NamedRerouteUsage()
  n83 = ComponentMask(None=VertexNormalWS()) {b=true}
  n84 = OneMinus(None=@n83)
  n85 = Multiply(A=@n84, B=Abs(None=@n83))
  n86 = BlendMaterialAttributes(
   A=@n82
   B=@n81
   Alpha=StaticSwitchParameter(True=Saturate(None=@n84), False=Saturate(None=@n85)) {group=RVT, parameterName="Skip RVT on Bottom && Sides"}
  )
  n87 = SetMaterialAttributes!(
   MaterialAttributes=StaticSwitchParameter(True=@n81, False=StaticSwitchParameter(True=@n86, False=@n82) {group=RVT, parameterName=Skip RVT on Bottom}) {group=RVT, parameterName=SkipRVT}
   Shading Model=ShadingModel()
  ) {attributeSetTypes [#6]}
  n88 = Multiply(A=Saturate(None=@n76.Result), B=Power(Base=@n77) {constExponent=0.5})
  n89 = MaterialFunctionCall!(
   Material=MaterialFunctionCall!(
    Base Material=@n87
    RoughnessMultiplier=MaterialFunctionCall(Input=@n78, Target Low=@n79, Target High=@n80) {materialFunction=&RemapValueRange}.Result
   ) {materialFunction=/Engine/Functions/MaterialLayerFunctions/MatLayerBlend_ModulateRoughness}.Blended Material
   NewBaseColor=MaterialFunctionCall(Input=OneMinus(None=@n88), Target Low=ScalarParameter() {defaultValue=0.375, parameterName=Wet Melting Snow Darkness}) {materialFunction=&RemapValueRange}.Result
   Mask
  ) {materialFunction=/Engine/Functions/MaterialLayerFunctions/MatLayerBlend_MultiplyBaseColor}
  n90 = MaterialFunctionCall(Input=Saturate(None=@n76.Result), Target Low=Constant() {r=1}, Target High=Constant() {r=1.256}) {materialFunction=&RemapValueRange}
  n91 = Constant() {r=0.095}
  n92 = MaterialFunctionCall(
   XY Tiling=Constant() {r=8}
   Z Scale=@n91
   Texture Object=TextureObject() {texture=/Game/Environments/Asteria/Sets/NeonCity/Textures/SnowWetnessDrips_M}
   Z Scale Gradient=@n91
  ) {materialFunction=&MF_XYPlanarWithTopScale}
  n93 = ComponentMask(None=@n92.Result) {r=true}
  n94 = MaterialFunctionCall!(
   "A (Subsurface)"=MaterialFunctionCall!(
    Mask=Constant() {r=1}
    SlipperySnow_Mask
    "Material;"=MaterialFunctionCall() {materialFunction=&MF_AsteriaLandscapeLayers}.Snow
   ) {materialFunction=/Game/Athena/Asteria/Environments/Landscape/Materials/MaterialFunctions/MF_SnowPostRVT}.SnowMaterial
   B=StaticSwitchParameter(
    True=MaterialFunctionCall!(Base Material=@n89.Blended Material, SpecularMultiplier=@n90.Result) {materialFunction=/Engine/Functions/MaterialLayerFunctions/MatLayerBlend_ModulateSpecular}.Blended Material
    False=@n87
   ) {group=TopMaterials, parameterName=Melting Snow}
   Subsurface Blend Min
   Subsurface Blend Max
   Alpha=StaticSwitchParameter(True=Multiply(A=@n92.Z Gradient, B=@n93), False=NamedRerouteUsage()) {group=TopMaterials, parameterName=Melting Snow}
  ) {materialFunction=/Game/Environments/Asteria/Rocks/Materials/MaterialFunctions/MF_BlendSubsurfaceMaterial}
  n95 = MaterialFunctionCall(
   NoRVT=@n73
   RVT=MaterialProxyReplace(Realtime=StaticSwitchParameter(True=@n94.Result MA, False=@n87) {group=TopMaterials, parameterName=Snow}, MaterialProxy=@n73)
  ) {materialFunction=&FN_RVT_Switch_Material}
  n96 = ShadingPathSwitch(Default=@n95.Result, Deferred=@n95.Result, Forward=@n73, Mobile=@n73)
  n97 = GetMaterialAttributes(None=@n96) {attributeGetTypes [#1]}
  n98 = TextureSampleParameter2D() {
   parameterName=MeshNormalMap
   samplerType=SAMPLERTYPE_Normal
   texture=/Game/Environments/Asteria/Props/Vista/Textures/BrokenIsland_Floating_NonNanite_N
  }
  n99 = Transform(None=@n98.RGB)
  n100 = Transform(None=@n97.Normal) {transformSourceType=TRANSFORMSOURCE_World, transformType=TRANSFORM_Tangent}
  StaticSwitchParameter(
   True=SetMaterialAttributes!(
    MaterialAttributes=@n97.MaterialAttributes
    Normal=MaterialFunctionCall(BaseNormal=@n99, AdditionalNormal=@n100) {materialFunction=&BlendAngleCorrectedNormals}.Result
   ) {attributeSetTypes [#1]}
   False=@n96
  ) {parameterName=UseMeshNormalMap}
  StaticSwitchParameter(True=@n6, False=@n4.RGB) {group=Fallback, parameterName="Use UV2 for FallBack "}
  n101 = Constant()
  n102 = StaticSwitchParameter(True=Constant() {r=1}, False=Constant()) {
   defaultValue=true
   desc="Set to true to spawn grass on this mesh.\r\nUses the vertex color to determine where to spawn."
   group=GrassSpawner
   parameterName="Use Grass(true) or a BaseTilingTexture(False)"
   sortPriority=0
  }
  n103 = SetMaterialAttributes!(
   MaterialAttributes=StaticSwitchParameter(True=SetMaterialAttributes!(MaterialAttributes, Opacity=@n101) {attributeSetTypes [#7]}, False=@n96) {group=HideMesh, parameterName=HideMeshUsingOpacity}
   Opacity Mask=@n102
  ) {attributeSetTypes [#8]}
  n104 = StaticBool() {value=true}
  n105 = StaticBool()
  n106 = StaticSwitchParameter(
   True=MaterialFunctionCall(Material_In=@n103, UseAdditiveWPO=StaticSwitchParameter(True=@n104, False=@n105) {parameterName=Use Additive WPO}) {materialFunction=&MF_FN_DestructibleDamage}.Material_Out
   False=@n103
  ) {defaultValue=true, parameterName=UseDestructableDamage}
  n107 = StaticSwitchParameter(
   True=MaterialFunctionCall!(In=@n106, UseSphericalMappingOverride) {materialFunction=/Game/Packages/Fortress_SharedMaterials/Base_Material_Functions/MF_CurieFallbackEffect}.Result
   False=@n106
  ) {defaultValue=true, parameterName=UseCurieFallBack}
  StaticSwitchParameter(
   True=MaterialFunctionCall(MA In=@n107) {materialFunction=&MF_CreativeMoveTool}.MA Out
   False=@n107
  ) {defaultValue=true, parameterName=UseCreativeMoveTool}
  StaticSwitchParameter(True=@n46, False=@n45) {
   desc="Disable to set up fallback material if using vertex colors, and enable back after that. "
   group=Ice
   parameterName=Use VertexColor Mask
   sortPriority=-1
  }
  StaticSwitchParameter(
   True=MaterialFunctionCall(BaseNormal=@n4.RGB, AdditionalNormal=NamedRerouteUsage()) {materialFunction=&BlendAngleCorrectedNormals}.Result
   False=@n8
  ) {group=Fallback, parameterName=UseUV0ForFallback}
 ]
}
